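# MetaSound patch that serves as a debug metronome, turning a MIDI clock input into an audible mono click track.
# Harmonix subdivision triggers fire on the base grid point (apparently the bar downbeat) and at three offsets after it, implying four beats.
# Each click is a square oscillator gated by a short AD envelope, higher-pitched for the downbeat and an octave lower for the merged offset triggers, then mixed to the output.
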MetaSoundPatch /FMJamSystem/Audio/MSP_DebugMetronome
inputs [MidiClock {type=MidiClock}]
outputs [Output {type=Audio}]
pages [Default {
  input_defaults {MidiClock=null}
  nodes [
   TriggerAnyTriggerAny3_0 = "TriggerAny.Trigger Any (3)"(
    In 1=HarmonixNodes.MidiClockSubdivisionTrigger(Midi Clock=^MidiClock) {also {defaults {Base Grid Size=0, "Base Offset Size "=1, Offset Multiplier=2}}}.Trigger Out
    In 0=HarmonixNodes.MidiClockSubdivisionTrigger(Midi Clock=^MidiClock) {also {defaults {Base Grid Size=0, "Base Offset Size "=1, Offset Multiplier=1}}}.Trigger Out
    In 2=HarmonixNodes.MidiClockSubdivisionTrigger(Midi Clock=^MidiClock) {also {defaults {Base Grid Size=0, "Base Offset Size "=1, Offset Multiplier=3}}}.Trigger Out
   )
   ADEnvelopeADEnvelope_1 = AD Envelope.AD Envelope:Audio(Trigger=@TriggerAnyTriggerAny3_0.Out) {also {defaults {Decay Time=0.2}}}
   ADEnvelopeADEnvelope_0 = AD Envelope.AD Envelope:Audio(Trigger=HarmonixNodes.MidiClockSubdivisionTrigger(Midi Clock=^MidiClock) {also {defaults {Base Grid Size=0, "Base Offset Size "=1}}}.Trigger Out) {also {defaults {Decay Time=0.2}}}
   AudioMixerAudioMixerMono2_0 = "AudioMixer.Audio Mixer (Mono, 2)"(
    In 1=UE.Multiply:Audio(PrimaryOperand=@ADEnvelopeADEnvelope_1.Out Envelope, AdditionalOperands=UE.Square:Audio() {also {defaults {Frequency=600, Type=0}}}.Audio).Out
    In 0=UE.Multiply:Audio(PrimaryOperand=@ADEnvelopeADEnvelope_0.Out Envelope, AdditionalOperands=UE.Square:Audio() {also {defaults {Frequency=1200, Type=0}}}.Audio).Out
   )
  ]
  links [{dst="@Output", src=AudioMixerAudioMixerMono2_0.Out}]
 }]
