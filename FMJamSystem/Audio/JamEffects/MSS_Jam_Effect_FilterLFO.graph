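# Stereo MetaSound jam effect that applies a tempo-synced, LFO-swept filter with waveshaping to Input L and Input R.
# Each channel runs through a waveshaper into a biquad filter, and IsActive selects dry or processed sound via a crossfade.
# A MIDI-clock-synced Fabric LFO drives both cutoffs through log-frequency mapping and smoothing, its beat duration a whole-number power of two derived from Rate.
# The Fabric AudioIn and AudioOut pins are unconnected.
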MetaSoundSource /FMJamSystem/Audio/JamEffects/MSS_Jam_Effect_FilterLFO
inputs [
 Fabric.Audio.AudioIn {type=Audio}
 Fabric.Common.MidiClockIn {type=MIDIClock}
 Fabric.Common.RootNote {type=Int32}
 Fabric.Common.Scale {type=Enum:MusicalScale}
 Fabric.Common.TransportIn {type=MusicTransport}
 Input L {type=Audio}
 Input R {type=Audio}
 IsActive {type=Bool}
 IsActiveTest {type=Float}
 Rate {type=Float}
 UE.Source.OnPlay {type=Trigger}
]
interfaces [Fabric.AudioEffect:1.0, UE.OutputFormat.Stereo:1.0, UE.Source:1.1]
outputs [
 Fabric.AudioEffect.AudioOut {type=Audio}
 "UE.OutputFormat.Stereo.Audio:0" {type=Audio}
 "UE.OutputFormat.Stereo.Audio:1" {type=Audio}
]
pages [Default {
  input_defaults {
   Fabric.Audio.AudioIn=null
   Fabric.Common.MidiClockIn=null
   Fabric.Common.RootNote=0
   Fabric.Common.Scale=0
   Fabric.Common.TransportIn=null
   Input L=null
   Input R=null
   IsActive=null
   IsActiveTest=0
   Rate=0
   UE.Source.OnPlay=null
  }
  nodes [
   Convert.Bool:Float(Float=^IsActiveTest)
   n1 = Convert.Float:Bool(Bool=^IsActive)
   n2 = UE.Reroute(Value=^Input R)
   n3 = Convert.Float:Int32(Int32=Convert.Int32:Float(Float=UE.Subtract:Float(AdditionalOperands=^Rate) {also {defaults {PrimaryOperand=4}}}.Out).Int32)
   LinearToLogFrequency_0 = UE.Linear To Log Frequency(Value=FabricNodes.FabricLFONode(
     MIDI Clock=^Fabric.Common.MidiClockIn
     Beat Sync Duration=UE.Power:Float(AdditionalOperands=@n3.Float) {also {defaults {PrimaryOperand=2}}}.Out
    ) {also {defaults {Shape=0}, version [0,1]}}.Output) {also {defaults {Max Range=5000, Min Range=300}}}
   n4 = UE.Reroute(Value=UE.Reroute(Value=UE.InterpTo:Audio(Target=@LinearToLogFrequency_0.Frequency) {also {defaults {Interp Time=0.05}}}.Value).Value)
   BiquadFilter_0 = UE.Biquad Filter:Audio(
    Cutoff Frequency=@n4.Value
    In=UE.WaveShaper:Audio(In=UE.Reroute(Value=@n2.Value).Value) {also {defaults {Amount=4, OutputGain=0.75, Type=4}}}.Out
   ) {also {defaults {Bandwidth=2, Gain=10, Type=2}}}
   CrossfadeTriggerRouteAudio2_0 = "Crossfade.Trigger Route (Audio, 2):Audio"(Crossfade Value=@n1.Float, In 0=UE.Reroute(Value=@n2.Value).Value, In 1=UE.Reroute(Value=@BiquadFilter_0.Out).Value)
   n5 = UE.Reroute(Value=^Input L)
   BiquadFilter_1 = UE.Biquad Filter:Audio(
    Cutoff Frequency=@n4.Value
    In=UE.WaveShaper:Audio(In=UE.Reroute(Value=@n5.Value).Value) {also {defaults {Amount=4, OutputGain=0.66, Type=4}}}.Out
   ) {also {defaults {Bandwidth=2, Gain=10, Type=2}}}
   CrossfadeTriggerRouteAudio2_1 = "Crossfade.Trigger Route (Audio, 2):Audio"(Crossfade Value=@n1.Float, In 0=UE.Reroute(Value=@n5.Value).Value, In 1=UE.Reroute(Value=@BiquadFilter_1.Out).Value)
  ]
  links [
   {dst="@UE.OutputFormat.Stereo.Audio:1", src=CrossfadeTriggerRouteAudio2_0.Out}
   {dst="@UE.OutputFormat.Stereo.Audio:0", src=CrossfadeTriggerRouteAudio2_1.Out}
  ]
 }]
properties {
 assetImportData=&self:AssetImportData
 bLooping=true
 duration=10000
 numChannels=2
 outputFormat=Stereo
 sampleRate=48000
}
