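# Looping stereo MetaSound audio effect for the Jam system that applies a tempo-synced stutter to incoming audio.
# Turning IsActive on starts a Harmonix StutterController through a MIDI-clock quantized trigger; turning it off sends an unquantized stop.
# Rate indexes an integer array to set the stutter subdivision, refreshed on start or Rate change.
# The controller drives one SimpleSampler per channel, with IsActive crossfading between dry input and sampler output.
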MetaSoundSource /FMJamSystem/Audio/JamEffects/MSS_Jam_Effect_StutterSimple
inputs [
 Fabric.Audio.AudioIn {type=Audio}
 Fabric.Common.MidiClockIn {type=MIDIClock}
 Fabric.Common.RootNote {type=Int32}
 Fabric.Common.Scale {type=Enum:MusicalScale}
 Fabric.Common.TransportIn {type=MusicTransport}
 Input L {type=Audio}
 Input R {type=Audio}
 IsActive {type=Bool}
 Rate {type=Float}
 UE.Source.OnPlay {type=Trigger}
]
interfaces [Fabric.AudioEffect:1.0, UE.OutputFormat.Stereo:1.0, UE.Source:1.1]
outputs [
 Fabric.AudioEffect.AudioOut {type=Audio}
 "UE.OutputFormat.Stereo.Audio:0" {type=Audio}
 "UE.OutputFormat.Stereo.Audio:1" {type=Audio}
]
pages [Default {
  input_defaults {
   Fabric.Audio.AudioIn=null
   Fabric.Common.MidiClockIn=null
   Fabric.Common.RootNote=0
   Fabric.Common.Scale=0
   Fabric.Common.TransportIn=null
   Input L=null
   Input R=null
   IsActive=null
   Rate=3
   UE.Source.OnPlay=null
  }
  nodes [
   n1 = Convert.Float:Bool(Bool=^IsActive)
   n2 = Convert.Int32:Float(Float=^Rate)
   n3 = UE.Reroute(Value=^Fabric.Common.MidiClockIn)
   n4 = UE.Reroute(Value=^IsActive)
   TriggerOnValueChange_0 = UE.Trigger On Value Change:Int32(Value=Convert.Int32:Bool(Bool=@n4.Value).Int32)
   n5 = TriggerCompare.Clamp:Bool(A=@n4.Value, Compare=@TriggerOnValueChange_0.Trigger) {also {defaults {B="<MetaSoundBuilderResult.SUCCEEDED: 0>"}}}
   n6 = HarmonixNodes.MidiQuantizeTriggerNode(MIDI Clock=@n3.Value, Trigger Input=@n5.True) {also {defaults {Subdivision=5, Tolerance=0}, version [0,1]}}
   TriggerOnValueChange_1 = UE.Trigger On Value Change:Int32(Value=@n2.Int32)
   TriggerAnyTriggerAny2_0 = "TriggerAny.Trigger Any (2)"(In 0=@n6.Trigger Out, In 1=@TriggerOnValueChange_1.Trigger)
   n7 = VariableAccessor.Int32:Array(Variable=InitVariable.Int32:Array().Variable)
   n8 = HarmonixNodes.StutterController(
    Spacing=Convert.Enum:SubdivisionQuantizationType:Int32(Int32=Array.Get:Int32:Array(Index=@n2.Int32, Trigger=@TriggerAnyTriggerAny2_0.Out, Array=@n7.Value).Element)."Enum:SubdivisionQuantizationType"
    Start=@n6.Trigger Out
    Stop=UE.Reroute(Value=UE.Reroute(Value=@n5.False).Value).Value
    MIDI Clock=UE.Reroute(Value=@n3.Value).Value
   ) {also {defaults {Audible Duration=6, Capture Duration=6}, version [0,1]}}
   n9 = UE.Reroute(Value=^Input L)
   CrossfadeTriggerRouteAudio2_0 = "Crossfade.Trigger Route (Audio, 2):Audio"(
    Crossfade Value=@n1.Float
    In 1=HarmonixNodes.SimpleSampler(
     Capture=@n8.Capture
     Capture Duration Sec=@n8.Capture Duration Sec
     Play=@n8.Play
     Play Duration Sec=@n8.Play Duration Sec
     Reset=@n8.Reset
     Reverse Playback=@n8.Reverse Playback
     In 0=@n9.Value
    ) {also {defaults {Enable="<MetaSoundBuilderResult.SUCCEEDED: 0>"}, version [0,1]}}.Out 0
    In 0=@n9.Value
   )
   n10 = UE.Reroute(Value=^Input R)
   CrossfadeTriggerRouteAudio2_1 = "Crossfade.Trigger Route (Audio, 2):Audio"(
    Crossfade Value=@n1.Float
    In 1=HarmonixNodes.SimpleSampler(
     Capture=@n8.Capture
     Capture Duration Sec=@n8.Capture Duration Sec
     Play=@n8.Play
     Play Duration Sec=@n8.Play Duration Sec
     Reset=@n8.Reset
     Reverse Playback=@n8.Reverse Playback
     In 0=@n10.Value
    ) {also {defaults {Enable="<MetaSoundBuilderResult.SUCCEEDED: 0>"}, version [0,1]}}.Out 0
    In 0=@n10.Value
   )
   InitVariable.StutterSequenceAsset:Array()
  ]
  links [
   {dst="@UE.OutputFormat.Stereo.Audio:0", src=CrossfadeTriggerRouteAudio2_0.Out}
   {dst="@UE.OutputFormat.Stereo.Audio:1", src=CrossfadeTriggerRouteAudio2_1.Out}
  ]
 }]
properties {
 assetImportData=&self:AssetImportData
 bLooping=true
 duration=10000
 numChannels=2
 outputFormat=Stereo
 sampleRate=48000
}
