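# MetaSound patch for one jam loop instance: plays a MIDI asset through a Fusion stereo sampler on the shared MIDI clock and transport.
# An ADSR envelope, output separately from the audio, releases on stop or loop ID change.
# Its attack comes from a quantized trigger, a "[pickup]" markup event or SkipQuantization, fed back through a deferred variable.
# A Trigger Once, re-armed on play or loop change, admits only the first such trigger.
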MetaSoundPatch /FMJamSystem/Audio/MSP_JamInstance
inputs [
 Global-MidiClock {type=MidiClock}
 Global-Transport {type=MusicTransport}
 Instance-Fusion {type=FusionPatchAsset}
 Instance-LoopInstanceId {type=Int32}
 Instance-MarkupMidiTrackIndex {type=Int32}
 Instance-Midi {type=MidiAsset}
 Instance-Playing {type=Bool}
 Instance-SkipQuantization {type=Trigger}
 Instance-Transpose {type=Int32}
 Multithreaded Rendering {type=Bool, access=Constructor}
 VoicePool {type=String, access=Constructor}
]
outputs [
 Out Envelope {type=Audio}
 Output-Left {type=Audio}
 Output-Right {type=Audio}
 Render Sync {type=Fusion Sync Link}
]
pages [Default {
  input_defaults {
   Global-MidiClock=null
   Global-Transport=null
   Instance-Fusion=null
   Instance-LoopInstanceId=-1
   Instance-MarkupMidiTrackIndex=-1
   Instance-Midi=null
   Instance-Playing=null
   Instance-SkipQuantization=null
   Instance-Transpose=0
   Multithreaded Rendering="<MetaSoundBuilderResult.SUCCEEDED: 0>"
   VoicePool=""
  }
  nodes [
   TriggerOnValueChange_1 = UE.Trigger On Value Change:Int32(Value=^Instance-LoopInstanceId)
   n1 = UE.Reroute(Value=@TriggerOnValueChange_1.Trigger)
   n2 = UE.Reroute(Value=^Instance-Playing)
   TriggerOnValueChange_0 = UE.Trigger On Value Change:Bool(Value=^Instance-Playing)
   n3 = TriggerCompare.Clamp:Bool(A=@n2.Value, Compare=@TriggerOnValueChange_0.Trigger) {also {defaults {B="<MetaSoundBuilderResult.SUCCEEDED: 0>"}}}
   TriggerAnyTriggerAny2_1 = "TriggerAny.Trigger Any (2)"(In 1=UE.Reroute(Value=@n1.Value).Value, In 0=@n3.False)
   TriggerAnyTriggerAny2_0 = "TriggerAny.Trigger Any (2)"(In 1=@n1.Value, In 0=@n3.True)
   n4 = VariableDeferredAccessor.Trigger(Variable=InitVariable.Trigger().Variable)
   TriggerOnce_0 = UE.Trigger Once(Reset=@TriggerAnyTriggerAny2_0.Out, Trigger In=@n4.Value)
   ADSREnvelopeADSREnvelope_0 = ADSR Envelope.ADSR Envelope:Audio(Trigger Release=@TriggerAnyTriggerAny2_1.Out, Trigger Attack=@TriggerOnce_0.Trigger Out) {also {defaults {Attack Time=0.005, Release Time=0.005, Sustain Level=1}}}
   n5 = HarmonixNodes.MIDIPlayer(MIDI Clock=^Global-MidiClock, Transport=^Global-Transport, Midi File=^Instance-Midi) {also {defaults {
      Kill Voices On Midi File Change="<MetaSoundBuilderResult.SUCCEEDED: 0>"
      Kill Voices On Seek="<MetaSoundBuilderResult.SUCCEEDED: 0>"
      Pre-roll Bars=64
     }}}
   n6 = UE.Reroute(Value=@n5.Midi Stream)
   HarmonixNodesFusionSamplerStereo_0 = HarmonixNodes.FusionSamplerStereo(
    Patch=^Instance-Fusion
    Transposition=^Instance-Transpose
    Multithreaded Rendering=^Multithreaded Rendering
    Voice Pool Name=^VoicePool
    Midi Stream=@n6.Value
   ) {also {defaults {Use Pitch Shifters="<MetaSoundBuilderResult.SUCCEEDED: 0>"}, version [0,1]}}
   PrintLogPrintLog_0 = Print Log.Print Log:Bool(Value To Log=^Instance-Playing) {also {defaults {Label="Instance-Playing: "}}}
   n7 = UE.Reroute(Value=^Instance-LoopInstanceId)
   PrintLogPrintLog_1 = Print Log.Print Log:Int32(Value To Log=@n7.Value) {also {defaults {Label="Instance-LoopInstanceId: "}}}
   PrintLogPrintLog_2 = Print Log.Print Log:Int32(Value To Log=^Instance-LoopInstanceId) {also {defaults {Label=Got Quantize Trigger}}}
   n8 = UE.Reroute(Value=@n5.MIDI Clock)
   TriggerAnyTriggerAny2_2 = "TriggerAny.Trigger Any (2)"(
    In 1=TriggerCompare.Clamp:Bool(A=^Instance-Playing, Compare=HarmonixNodes.TransportToTrigger(Transport=^Global-Transport) {also {version [0,1]}}.Play) {also {defaults {B="<MetaSoundBuilderResult.SUCCEEDED: 0>"}}}.True
    In 0=TriggerCompare.Clamp:Int32(A=UE.Reroute(Value=@n7.Value).Value, Compare=@TriggerOnValueChange_1.Trigger) {also {defaults {B=-1, Type=1}}}.True
   )
   n9 = UE.Reroute(Value=@n6.Value)
   TriggerAnyTriggerAny3_0 = "TriggerAny.Trigger Any (3)"(
    In 2=^Instance-SkipQuantization
    In 0=HarmonixNodes.MidiQuantizeTriggerNode(MIDI Clock=UE.Reroute(Value=@n8.Value).Value, Trigger Input=@TriggerAnyTriggerAny2_2.Out) {also {version [0,1]}}.Trigger Out
    In 1=HarmonixNodes.MidiTextTrigger(
     MIDI Stream=HarmonixNodes.MidiStreamTrackFilter(Max. Track Index=^Instance-MarkupMidiTrackIndex, Min. Track Index=^Instance-MarkupMidiTrackIndex, MIDI Stream=@n9.Value).MIDI Stream
     Enable=UE.Reroute(Value=@n2.Value).Value
    ) {also {defaults {Text="[pickup]"}}}.Trigger Out
   )
   VariableMutator.Trigger(Value=@TriggerAnyTriggerAny3_0.Out, Variable=@n4.Variable)
  ]
  links [
   {dst="@Out Envelope", src=ADSREnvelopeADSREnvelope_0.Out Envelope}
   {dst="@Output-Left", src=HarmonixNodesFusionSamplerStereo_0.Audio Left}
   {dst="@Output-Right", src=HarmonixNodesFusionSamplerStereo_0.Audio Right}
   {dst="@Render Sync", src=HarmonixNodesFusionSamplerStereo_0.Render Sync}
  ]
 }]
