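# MetaSound patch for the Jam music player, rendering four MIDI/Fusion loop instances with built-in effects into stereo.
# Synchronized instance outputs pass through clocked Stutter and FilterLFO effects, envelope multiplication, then per-instance and global gain mixers.
# Pre-gain peak and multiband analyzers output per-instance reactivity levels, with bands sampled by a periodic trigger; overall amplitude precedes global gain.
# FX2 and FX3 inputs are declared but unconnected.
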MetaSoundPatch /FMJamSystem/Audio/MSP_JamMusicPlayer_InternalFX
inputs [
 Crossover Frequencies {type=Float:Array, access=Constructor}
 FusionVoicePool {type=String, access=Constructor}
 Global-Gain {type=Float}
 J0-Fusion {type=FusionPatchAsset}
 J0-FX0-IsActive {type=Bool}
 J0-FX0-Param {type=Float}
 J0-FX1-IsActive {type=Bool}
 J0-FX1-Param {type=Float}
 J0-FX2-IsActive {type=Bool}
 J0-FX2-Param {type=Float}
 J0-FX3-IsActive {type=Bool}
 J0-FX3-Param {type=Float}
 J0-Gain {type=Float}
 J0-LoopInstanceId {type=Int32}
 J0-MarkupMidiTrackIndex {type=Int32}
 J0-Midi {type=MidiAsset}
 J0-Playing {type=Bool}
 J0-Transpose {type=Int32}
 J1-Fusion {type=FusionPatchAsset}
 J1-FX0-IsActive {type=Bool}
 J1-FX0-Param {type=Float}
 J1-FX1-IsActive {type=Bool}
 J1-FX1-Param {type=Float}
 J1-FX2-IsActive {type=Bool}
 J1-FX2-Param {type=Float}
 J1-FX3-IsActive {type=Bool}
 J1-FX3-Param {type=Float}
 J1-Gain {type=Float}
 J1-LoopInstanceId {type=Int32}
 J1-MarkupMidiTrackIndex {type=Int32}
 J1-Midi {type=MidiAsset}
 J1-Playing {type=Bool}
 J1-Transpose {type=Int32}
 J2-Fusion {type=FusionPatchAsset}
 J2-FX0-IsActive {type=Bool}
 J2-FX0-Param {type=Float}
 J2-FX1-IsActive {type=Bool}
 J2-FX1-Param {type=Float}
 J2-FX2-IsActive {type=Bool}
 J2-FX2-Param {type=Float}
 J2-FX3-IsActive {type=Bool}
 J2-FX3-Param {type=Float}
 J2-Gain {type=Float}
 J2-LoopInstanceId {type=Int32}
 J2-MarkupMidiTrackIndex {type=Int32}
 J2-Midi {type=MidiAsset}
 J2-Playing {type=Bool}
 J2-Transpose {type=Int32}
 J3-Fusion {type=FusionPatchAsset}
 J3-FX0-IsActive {type=Bool}
 J3-FX0-Param {type=Float}
 J3-FX1-IsActive {type=Bool}
 J3-FX1-Param {type=Float}
 J3-FX2-IsActive {type=Bool}
 J3-FX2-Param {type=Float}
 J3-FX3-IsActive {type=Bool}
 J3-FX3-Param {type=Float}
 J3-Gain {type=Float}
 J3-LoopInstanceId {type=Int32}
 J3-MarkupMidiTrackIndex {type=Int32}
 J3-Midi {type=MidiAsset}
 J3-Playing {type=Bool}
 J3-Transpose {type=Int32}
 MidiClockIn {type=MidiClock}
 Multithreaded Rendering {type=Bool, access=Constructor}
 ReactivityUpdatePeriod {type=Time}
 TransportIn {type=MusicTransport}
]
outputs [
 AmplitudeOut {type=Float}
 J0-Amplitude {type=Float}
 J0-Band1 {type=Float}
 J0-Band2 {type=Float}
 J0-Band3 {type=Float}
 J0-Band4 {type=Float}
 J1-Amplitude {type=Float}
 J1-Band1 {type=Float}
 J1-Band2 {type=Float}
 J1-Band3 {type=Float}
 J1-Band4 {type=Float}
 J2-Amplitude {type=Float}
 J2-Band1 {type=Float}
 J2-Band2 {type=Float}
 J2-Band3 {type=Float}
 J2-Band4 {type=Float}
 J3-Amplitude {type=Float}
 J3-Band1 {type=Float}
 J3-Band2 {type=Float}
 J3-Band3 {type=Float}
 J3-Band4 {type=Float}
 Out Left {type=Audio}
 Out Right {type=Audio}
]
pages [Default {
  input_defaults {
   FusionVoicePool=""
   Global-Gain=1
   J0-FX0-IsActive=null
   J0-FX0-Param=0
   J0-FX1-IsActive=null
   J0-FX1-Param=0
   J0-FX2-IsActive=null
   J0-FX2-Param=0
   J0-FX3-IsActive=null
   J0-FX3-Param=0
   J0-Fusion=null
   J0-Gain=1
   J0-LoopInstanceId=-1
   J0-MarkupMidiTrackIndex=-1
   J0-Midi=null
   J0-Playing=null
   J0-Transpose=0
   J1-FX0-IsActive=null
   J1-FX0-Param=0
   J1-FX1-IsActive=null
   J1-FX1-Param=0
   J1-FX2-IsActive=null
   J1-FX2-Param=0
   J1-FX3-IsActive=null
   J1-FX3-Param=0
   J1-Fusion=null
   J1-Gain=1
   J1-LoopInstanceId=-1
   J1-MarkupMidiTrackIndex=-1
   J1-Midi=null
   J1-Playing=null
   J1-Transpose=0
   J2-FX0-IsActive=null
   J2-FX0-Param=0
   J2-FX1-IsActive=null
   J2-FX1-Param=0
   J2-FX2-IsActive=null
   J2-FX2-Param=0
   J2-FX3-IsActive=null
   J2-FX3-Param=0
   J2-Fusion=null
   J2-Gain=1
   J2-LoopInstanceId=-1
   J2-MarkupMidiTrackIndex=-1
   J2-Midi=null
   J2-Playing=null
   J2-Transpose=0
   J3-FX0-IsActive=null
   J3-FX0-Param=0
   J3-FX1-IsActive=null
   J3-FX1-Param=0
   J3-FX2-IsActive=null
   J3-FX2-Param=0
   J3-FX3-IsActive=null
   J3-FX3-Param=0
   J3-Fusion=null
   J3-Gain=1
   J3-LoopInstanceId=-1
   J3-MarkupMidiTrackIndex=-1
   J3-Midi=null
   J3-Playing=null
   J3-Transpose=0
   MidiClockIn=null
   Multithreaded Rendering="<MetaSoundBuilderResult.SUCCEEDED: 0>"
   ReactivityUpdatePeriod=0.0333
   TransportIn=null
   Crossover Frequencies [200,700,1200]
  }
  nodes [
   n1 = UE.Reroute(Value=^Multithreaded Rendering)
   n2 = UE.Reroute(Value=@n1.Value)
   n3 = UE.Reroute(Value=^TransportIn)
   n4 = UE.Reroute(Value=@n3.Value)
   n5 = UE.Reroute(Value=^FusionVoicePool)
   n6 = UE.Reroute(Value=@n5.Value)
   n7 = UE.Reroute(Value=^MidiClockIn)
   n8 = UE.Reroute(Value=@n7.Value)
   n9 = &/MSP_JamInstance(
    Instance-Fusion=^J1-Fusion
    Instance-LoopInstanceId=^J1-LoopInstanceId
    Instance-MarkupMidiTrackIndex=^J1-MarkupMidiTrackIndex
    Instance-Midi=^J1-Midi
    Instance-Playing=^J1-Playing
    Instance-Transpose=^J1-Transpose
    Multithreaded Rendering=@n2.Value
    Global-Transport=@n4.Value
    VoicePool=@n6.Value
    Global-MidiClock=@n8.Value
   )
   n10 = UE.Reroute(Value=UE.Reroute(Value=@n5.Value).Value)
   n11 = UE.Reroute(Value=UE.Reroute(Value=@n1.Value).Value)
   n12 = UE.Reroute(Value=UE.Reroute(Value=@n3.Value).Value)
   n13 = UE.Reroute(Value=UE.Reroute(Value=@n7.Value).Value)
   n14 = &MSP_JamInstance(
    Instance-Fusion=^J3-Fusion
    Instance-LoopInstanceId=^J3-LoopInstanceId
    Instance-MarkupMidiTrackIndex=^J3-MarkupMidiTrackIndex
    Instance-Midi=^J3-Midi
    Instance-Playing=^J3-Playing
    Instance-Transpose=^J3-Transpose
    VoicePool=@n10.Value
    Multithreaded Rendering=@n11.Value
    Global-Transport=@n12.Value
    Global-MidiClock=@n13.Value
   )
   n15 = &MSP_JamInstance(
    Instance-Fusion=^J0-Fusion
    Instance-LoopInstanceId=^J0-LoopInstanceId
    Instance-MarkupMidiTrackIndex=^J0-MarkupMidiTrackIndex
    Instance-Midi=^J0-Midi
    Instance-Playing=^J0-Playing
    Instance-Transpose=^J0-Transpose
    Multithreaded Rendering=@n2.Value
    Global-Transport=@n4.Value
    VoicePool=@n6.Value
    Global-MidiClock=@n8.Value
   )
   n16 = &MSP_JamInstance(
    Instance-Fusion=^J2-Fusion
    Instance-LoopInstanceId=^J2-LoopInstanceId
    Instance-MarkupMidiTrackIndex=^J2-MarkupMidiTrackIndex
    Instance-Midi=^J2-Midi
    Instance-Playing=^J2-Playing
    Instance-Transpose=^J2-Transpose
    VoicePool=@n10.Value
    Multithreaded Rendering=@n11.Value
    Global-Transport=@n12.Value
    Global-MidiClock=@n13.Value
   )
   n17 = HarmonixNodes.4InstanceStereoFusionSynchronizer(
    Sampler 2 In L=@n9.Output-Left
    Sampler 2 In R=@n9.Output-Right
    Sampler 2 Sync=@n9.Render Sync
    Sampler 4 In L=@n14.Output-Left
    Sampler 4 In R=@n14.Output-Right
    Sampler 4 Sync=@n14.Render Sync
    Sampler 1 In L=@n15.Output-Left
    Sampler 1 In R=@n15.Output-Right
    Sampler 1 Sync=@n15.Render Sync
    Sampler 3 In L=@n16.Output-Left
    Sampler 3 In R=@n16.Output-Right
    Sampler 3 Sync=@n16.Render Sync
   ) {also {version [0,1]}}
   n18 = UE.Reroute(Value=UE.Reroute(Value=UE.Reroute(Value=@n7.Value).Value).Value)
   n19 = UE.Reroute(Value=@n18.Value)
   n20 = UE.Reroute(Value=@n19.Value)
   n21 = /FMJamSystem/Audio/JamEffects/MSS_Jam_Effect_StutterSimple(
    IsActive=^J3-FX0-IsActive
    Rate=^J3-FX0-Param
    Input R=UE.Reroute(Value=@n17.Sampler 4 Out R).Value
    Fabric.Common.MidiClockIn=UE.Reroute(Value=@n20.Value).Value
    Input L=UE.Reroute(Value=@n17.Sampler 4 Out L).Value
   )
   n22 = UE.Reroute(Value=UE.Reroute(Value=@n18.Value).Value)
   n23 = UE.Reroute(Value=@n22.Value)
   n24 = UE.Reroute(Value=@n23.Value)
   n25 = /FMJamSystem/Audio/JamEffects/MSS_Jam_Effect_FilterLFO(
    IsActive=^J3-FX1-IsActive
    Rate=^J3-FX1-Param
    Input L=@n21."UE.OutputFormat.Stereo.Audio:0"
    Input R=@n21."UE.OutputFormat.Stereo.Audio:1"
    Fabric.Common.MidiClockIn=@n24.Value
   )
   n26 = UE.Reroute(Value=UE.Reroute(Value=UE.Reroute(Value=@n14.Out Envelope).Value).Value)
   n27 = UE.Multiply:Audio(PrimaryOperand=@n25."UE.OutputFormat.Stereo.Audio:0", AdditionalOperands=@n26.Value)
   n28 = UE.Reroute(Value=@n27.Out)
   n29 = &MSS_Jam_Effect_StutterSimple(
    IsActive=^J0-FX0-IsActive
    Rate=^J0-FX0-Param
    Fabric.Common.MidiClockIn=@n18.Value
    Input L=UE.Reroute(Value=@n17.Sampler 1 Out L).Value
    Input R=UE.Reroute(Value=@n17.Sampler 1 Out R).Value
   )
   n30 = &MSS_Jam_Effect_FilterLFO(
    IsActive=^J0-FX1-IsActive
    Rate=^J0-FX1-Param
    Input L=@n29."UE.OutputFormat.Stereo.Audio:0"
    Input R=@n29."UE.OutputFormat.Stereo.Audio:1"
    Fabric.Common.MidiClockIn=@n22.Value
   )
   n31 = UE.Reroute(Value=UE.Reroute(Value=UE.Reroute(Value=@n15.Out Envelope).Value).Value)
   n32 = UE.Multiply:Audio(PrimaryOperand=@n30."UE.OutputFormat.Stereo.Audio:0", AdditionalOperands=@n31.Value)
   n33 = UE.Reroute(Value=@n32.Out)
   n34 = UE.Multiply:Audio(PrimaryOperand=@n30."UE.OutputFormat.Stereo.Audio:1", AdditionalOperands=@n31.Value)
   n35 = UE.Reroute(Value=@n34.Out)
   n36 = &MSS_Jam_Effect_StutterSimple(
    IsActive=^J2-FX0-IsActive
    Rate=^J2-FX0-Param
    Input L=UE.Reroute(Value=@n17.Sampler 3 Out L).Value
    Fabric.Common.MidiClockIn=@n20.Value
    Input R=UE.Reroute(Value=@n17.Sampler 3 Out R).Value
   )
   n37 = &MSS_Jam_Effect_FilterLFO(
    IsActive=^J2-FX1-IsActive
    Rate=^J2-FX1-Param
    Input L=@n36."UE.OutputFormat.Stereo.Audio:0"
    Input R=@n36."UE.OutputFormat.Stereo.Audio:1"
    Fabric.Common.MidiClockIn=@n24.Value
   )
   n38 = UE.Reroute(Value=UE.Reroute(Value=UE.Reroute(Value=@n16.Out Envelope).Value).Value)
   n39 = UE.Multiply:Audio(PrimaryOperand=@n37."UE.OutputFormat.Stereo.Audio:1", AdditionalOperands=@n38.Value)
   n40 = UE.Reroute(Value=@n39.Out)
   n41 = &MSS_Jam_Effect_StutterSimple(
    IsActive=^J1-FX0-IsActive
    Rate=^J1-FX0-Param
    Input L=UE.Reroute(Value=@n17.Sampler 2 Out L).Value
    Input R=UE.Reroute(Value=@n17.Sampler 2 Out R).Value
    Fabric.Common.MidiClockIn=@n19.Value
   )
   n42 = &MSS_Jam_Effect_FilterLFO(
    IsActive=^J1-FX1-IsActive
    Rate=^J1-FX1-Param
    Input L=@n41."UE.OutputFormat.Stereo.Audio:0"
    Input R=@n41."UE.OutputFormat.Stereo.Audio:1"
    Fabric.Common.MidiClockIn=@n23.Value
   )
   n43 = UE.Reroute(Value=UE.Reroute(Value=UE.Reroute(Value=@n9.Out Envelope).Value).Value)
   n44 = UE.Multiply:Audio(PrimaryOperand=@n42."UE.OutputFormat.Stereo.Audio:1", AdditionalOperands=@n43.Value)
   n45 = UE.Reroute(Value=@n44.Out)
   n46 = UE.Multiply:Audio(PrimaryOperand=@n42."UE.OutputFormat.Stereo.Audio:0", AdditionalOperands=@n43.Value)
   n47 = UE.Reroute(Value=@n46.Out)
   n48 = UE.Multiply:Audio(PrimaryOperand=@n37."UE.OutputFormat.Stereo.Audio:0", AdditionalOperands=@n38.Value)
   n49 = UE.Reroute(Value=@n48.Out)
   n50 = UE.Multiply:Audio(PrimaryOperand=@n25."UE.OutputFormat.Stereo.Audio:1", AdditionalOperands=@n26.Value)
   n51 = UE.Reroute(Value=@n50.Out)
   AudioMixerAudioMixerStereo4_0 = "AudioMixer.Audio Mixer (Stereo, 4)"(
    Gain 0=^J0-Gain
    Gain 1=^J1-Gain
    Gain 2=^J2-Gain
    Gain 3=^J3-Gain
    In 3 L=UE.Reroute(Value=@n28.Value).Value
    In 0 L=@n33.Value
    In 0 R=@n35.Value
    In 2 R=UE.Reroute(Value=@n40.Value).Value
    In 1 R=UE.Reroute(Value=@n45.Value).Value
    In 1 L=UE.Reroute(Value=@n47.Value).Value
    In 2 L=UE.Reroute(Value=@n49.Value).Value
    In 3 R=UE.Reroute(Value=@n51.Value).Value
   )
   AudioMixerAudioMixerStereo2_0 = "AudioMixer.Audio Mixer (Stereo, 2)"(Gain 0=^Global-Gain, In 0 L=@AudioMixerAudioMixerStereo4_0.Out L, In 0 R=@AudioMixerAudioMixerStereo4_0.Out R)
   Divide_0 = UE.Divide:Float(PrimaryOperand=HarmonixNodes.Peak(Audio Mono=UE.Add:Audio(PrimaryOperand=@n28.Value, AdditionalOperands=@n51.Value).Out) {also {version [0,1]}}.Peak) {also {defaults {AdditionalOperands=2}}}
   n52 = HarmonixNodes.MultibandAnalyzer(Crossover Frequencies=^Crossover Frequencies, Audio Mono=UE.Add:Audio(AdditionalOperands=@n34.Out, PrimaryOperand=@n32.Out).Out) {also {defaults {Release Time=0.1}, version [0,1]}}
   TriggerAnyTriggerAny2_0 = "TriggerAny.Trigger Any (2)"(
    In 0=HarmonixNodes.MidiClockSubdivisionTrigger(MIDI Clock=@n7.Value).Trigger Out
    In 1=HarmonixNodes.TransportToTrigger(Transport=@n3.Value) {also {version [0,1]}}.Play
   )
   TriggerOnce_0 = UE.Trigger Once(Trigger In=@TriggerAnyTriggerAny2_0.Out)
   n53 = UE.TriggerRepeat(Period=^ReactivityUpdatePeriod, Start=@TriggerOnce_0.Trigger Out)
   n54 = UE.Reroute(Value=@n53.RepeatOut)
   Divide_1 = UE.Divide:Float(PrimaryOperand=Array.Get:Float:Array(Array=@n52.Band Levels, Trigger=@n54.Value) {also {defaults {Index=1}}}.Element) {also {defaults {AdditionalOperands=2}}}
   n55 = HarmonixNodes.MultibandAnalyzer(Crossover Frequencies=^Crossover Frequencies, Audio Mono=UE.Add:Audio(PrimaryOperand=@n46.Out, AdditionalOperands=@n44.Out).Out) {also {defaults {Release Time=0.1}, version [0,1]}}
   n56 = UE.Reroute(Value=@n53.RepeatOut)
   Divide_10 = UE.Divide:Float(PrimaryOperand=Array.Get:Float:Array(Array=@n55.Band Levels, Trigger=@n56.Value) {also {defaults {Index=2}}}.Element) {also {defaults {AdditionalOperands=2}}}
   n57 = HarmonixNodes.MultibandAnalyzer(Crossover Frequencies=^Crossover Frequencies, Audio Mono=UE.Add:Audio(PrimaryOperand=@n27.Out, AdditionalOperands=@n50.Out).Out) {also {defaults {Release Time=0.1}, version [0,1]}}
   n58 = UE.Reroute(Value=@n53.RepeatOut)
   Divide_11 = UE.Divide:Float(PrimaryOperand=Array.Get:Float:Array(Array=@n57.Band Levels, Trigger=@n58.Value) {also {defaults {Index=3}}}.Element) {also {defaults {AdditionalOperands=2}}}
   n59 = HarmonixNodes.MultibandAnalyzer(Crossover Frequencies=^Crossover Frequencies, Audio Mono=UE.Add:Audio(AdditionalOperands=@n39.Out, PrimaryOperand=@n48.Out).Out) {also {defaults {Release Time=0.1}, version [0,1]}}
   n60 = UE.Reroute(Value=@n53.RepeatOut)
   Divide_12 = UE.Divide:Float(PrimaryOperand=Array.Get:Float:Array(Array=@n59.Band Levels, Trigger=@n60.Value) {also {defaults {Index=0}}}.Element) {also {defaults {AdditionalOperands=2}}}
   Divide_13 = UE.Divide:Float(PrimaryOperand=HarmonixNodes.Peak(Audio Mono=UE.Add:Audio(PrimaryOperand=@n49.Value, AdditionalOperands=@n40.Value).Out) {also {version [0,1]}}.Peak) {also {defaults {AdditionalOperands=2}}}
   Divide_14 = UE.Divide:Float(PrimaryOperand=HarmonixNodes.Peak(Audio Mono=UE.Add:Audio(PrimaryOperand=@AudioMixerAudioMixerStereo4_0.Out L, AdditionalOperands=@AudioMixerAudioMixerStereo4_0.Out R).Out) {also {version [0,1]}}.Peak) {also {defaults {AdditionalOperands=2}}}
   Divide_15 = UE.Divide:Float(PrimaryOperand=Array.Get:Float:Array(Array=@n59.Band Levels, Trigger=@n60.Value) {also {defaults {Index=3}}}.Element) {also {defaults {AdditionalOperands=2}}}
   Divide_16 = UE.Divide:Float(PrimaryOperand=Array.Get:Float:Array(Array=@n52.Band Levels, Trigger=@n54.Value) {also {defaults {Index=0}}}.Element) {also {defaults {AdditionalOperands=2}}}
   Divide_17 = UE.Divide:Float(PrimaryOperand=Array.Get:Float:Array(Array=@n57.Band Levels, Trigger=@n58.Value) {also {defaults {Index=0}}}.Element) {also {defaults {AdditionalOperands=2}}}
   Divide_18 = UE.Divide:Float(PrimaryOperand=HarmonixNodes.Peak(Audio Mono=UE.Add:Audio(AdditionalOperands=@n45.Value, PrimaryOperand=@n47.Value).Out) {also {version [0,1]}}.Peak) {also {defaults {AdditionalOperands=2}}}
   Divide_19 = UE.Divide:Float(PrimaryOperand=Array.Get:Float:Array(Array=@n57.Band Levels, Trigger=@n58.Value) {also {defaults {Index=2}}}.Element) {also {defaults {AdditionalOperands=2}}}
   Divide_2 = UE.Divide:Float(PrimaryOperand=Array.Get:Float:Array(Array=@n59.Band Levels, Trigger=@n60.Value) {also {defaults {Index=2}}}.Element) {also {defaults {AdditionalOperands=2}}}
   Divide_20 = UE.Divide:Float(PrimaryOperand=Array.Get:Float:Array(Array=@n55.Band Levels, Trigger=@n56.Value) {also {defaults {Index=1}}}.Element) {also {defaults {AdditionalOperands=2}}}
   Divide_3 = UE.Divide:Float(PrimaryOperand=Array.Get:Float:Array(Array=@n55.Band Levels, Trigger=@n56.Value) {also {defaults {Index=0}}}.Element) {also {defaults {AdditionalOperands=2}}}
   Divide_4 = UE.Divide:Float(PrimaryOperand=Array.Get:Float:Array(Array=@n57.Band Levels, Trigger=@n58.Value) {also {defaults {Index=1}}}.Element) {also {defaults {AdditionalOperands=2}}}
   Divide_5 = UE.Divide:Float(PrimaryOperand=Array.Get:Float:Array(Array=@n59.Band Levels, Trigger=@n60.Value) {also {defaults {Index=1}}}.Element) {also {defaults {AdditionalOperands=2}}}
   Divide_6 = UE.Divide:Float(PrimaryOperand=Array.Get:Float:Array(Array=@n55.Band Levels, Trigger=@n56.Value) {also {defaults {Index=3}}}.Element) {also {defaults {AdditionalOperands=2}}}
   Divide_7 = UE.Divide:Float(PrimaryOperand=HarmonixNodes.Peak(Audio Mono=UE.Add:Audio(PrimaryOperand=@n33.Value, AdditionalOperands=@n35.Value).Out) {also {version [0,1]}}.Peak) {also {defaults {AdditionalOperands=2}}}
   Divide_8 = UE.Divide:Float(PrimaryOperand=Array.Get:Float:Array(Array=@n52.Band Levels, Trigger=@n54.Value) {also {defaults {Index=2}}}.Element) {also {defaults {AdditionalOperands=2}}}
   Divide_9 = UE.Divide:Float(PrimaryOperand=Array.Get:Float:Array(Array=@n52.Band Levels, Trigger=@n54.Value) {also {defaults {Index=3}}}.Element) {also {defaults {AdditionalOperands=2}}}
   InitVariable.MIDIClock()
  ]
  links [
   {dst="@Out Left", src=AudioMixerAudioMixerStereo2_0.Out L}
   {dst="@Out Right", src=AudioMixerAudioMixerStereo2_0.Out R}
   {dst="@J3-Amplitude", src=Divide_0.Out}
   {dst="@J0-Band2", src=Divide_1.Out}
   {dst="@J1-Band3", src=Divide_10.Out}
   {dst="@J3-Band4", src=Divide_11.Out}
   {dst="@J2-Band1", src=Divide_12.Out}
   {dst="@J2-Amplitude", src=Divide_13.Out}
   {dst="@AmplitudeOut", src=Divide_14.Out}
   {dst="@J2-Band4", src=Divide_15.Out}
   {dst="@J0-Band1", src=Divide_16.Out}
   {dst="@J3-Band1", src=Divide_17.Out}
   {dst="@J1-Amplitude", src=Divide_18.Out}
   {dst="@J3-Band3", src=Divide_19.Out}
   {dst="@J2-Band3", src=Divide_2.Out}
   {dst="@J1-Band2", src=Divide_20.Out}
   {dst="@J1-Band1", src=Divide_3.Out}
   {dst="@J3-Band2", src=Divide_4.Out}
   {dst="@J2-Band2", src=Divide_5.Out}
   {dst="@J1-Band4", src=Divide_6.Out}
   {dst="@J0-Amplitude", src=Divide_7.Out}
   {dst="@J0-Band3", src=Divide_8.Out}
   {dst="@J0-Band4", src=Divide_9.Out}
  ]
 }]
