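# Jam_Main is a root SoundClass in the FMJamSystem plugin, categorizing audio for the Jam system.
# It has no parent class, applies ambient volumes, disables reverb, and routes its sounds by default to the Fortnite_SFX submix.
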SoundClass /FMJamSystem/Audio/Jam_Main
parent=null
properties {properties {bApplyAmbientVolumes=true, bReverb=false, defaultSubmix=&Fortnite_SFX}}
