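# Looping stereo MetaSound source that plays a four-part jam music mix (J0–J3) from MIDI and Fusion patch inputs.
# A netsync-seekable transport and metronome drive an empty 32-bar MIDI loop, whose clock feeds MSP_JamMusicPlayer_Internal.
# That patch's stereo mix passes through a StutteRecapture effect, controlled by J0-FX0 inputs, then an interpolated gain.
# Per-part amplitude and band outputs expose reactivity data; J4–J5 outputs are unwired.
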MetaSoundSource /FMJamSystem/Audio/MSS_JamMusicPlayer_StrideMice
inputs [
 Crossover Frequencies {type=Float:Array, access=Constructor}
 DefaultGainModifier {type=Float}
 Global-DebugMetronomeGain {type=Float}
 Global-Gain {type=Float}
 Global-Play {type=Trigger}
 Global-Stop {type=Trigger}
 Global-Tempo {type=Float}
 J0-Fusion {type=FusionPatchAsset}
 J0-FX0-IsActive {type=Bool}
 J0-FX0-Param {type=Float}
 J0-Gain {type=Float}
 J0-LoopInstanceId {type=Int32}
 J0-MarkupMidiTrackIndex {type=Int32}
 J0-Midi {type=MidiAsset}
 J0-Playing {type=Bool}
 J0-SkipQuantization {type=Trigger}
 J0-Transpose {type=Int32}
 J1-Fusion {type=FusionPatchAsset}
 J1-Gain {type=Float}
 J1-LoopInstanceId {type=Int32}
 J1-MarkupMidiTrackIndex {type=Int32}
 J1-Midi {type=MidiAsset}
 J1-Playing {type=Bool}
 J1-SkipQuantization {type=Trigger}
 J1-Transpose {type=Int32}
 J2-Fusion {type=FusionPatchAsset}
 J2-Gain {type=Float}
 J2-LoopInstanceId {type=Int32}
 J2-MarkupMidiTrackIndex {type=Int32}
 J2-Midi {type=MidiAsset}
 J2-Playing {type=Bool}
 J2-SkipQuantization {type=Trigger}
 J2-Transpose {type=Int32}
 J3-Fusion {type=FusionPatchAsset}
 J3-Gain {type=Float}
 J3-LoopInstanceId {type=Int32}
 J3-MarkupMidiTrackIndex {type=Int32}
 J3-Midi {type=MidiAsset}
 J3-Playing {type=Bool}
 J3-SkipQuantization {type=Trigger}
 J3-Transpose {type=Int32}
 Multithreaded Rendering {type=Bool, access=Constructor}
 ReactivityUpdatePeriod {type=Time}
 Sparks.MusicTimeNetSync.Seek {type=Trigger}
 Sparks.MusicTimeNetSync.Seek Target {type=MusicSeekTarget}
 Sparks.MusicTimeNetSync.Speed {type=Float}
 UE.Source.OnPlay {type=Trigger}
]
interfaces [Sparks.MusicTimeNetSync:1.0, UE.OutputFormat.Stereo:1.0, UE.Source:1.1]
outputs [
 AmplitudeOut {type=Float}
 ClockOut {type=MidiClock}
 J0-Amplitude {type=Float}
 J0-Band1 {type=Float}
 J0-Band2 {type=Float}
 J0-Band3 {type=Float}
 J0-Band4 {type=Float}
 J1-Amplitude {type=Float}
 J1-Band1 {type=Float}
 J1-Band2 {type=Float}
 J1-Band3 {type=Float}
 J1-Band4 {type=Float}
 J2-Amplitude {type=Float}
 J2-Band1 {type=Float}
 J2-Band2 {type=Float}
 J2-Band3 {type=Float}
 J2-Band4 {type=Float}
 J3-Amplitude {type=Float}
 J3-Band1 {type=Float}
 J3-Band2 {type=Float}
 J3-Band3 {type=Float}
 J3-Band4 {type=Float}
 J4-Amplitude {type=Float}
 J4-Band1 {type=Float}
 J4-Band2 {type=Float}
 J4-Band3 {type=Float}
 J4-Band4 {type=Float}
 J5-Amplitude {type=Float}
 J5-Band1 {type=Float}
 J5-Band2 {type=Float}
 J5-Band3 {type=Float}
 J5-Band4 {type=Float}
 Out Looped Midi Clock {type=MidiClock}
 "UE.OutputFormat.Stereo.Audio:0" {type=Audio}
 "UE.OutputFormat.Stereo.Audio:1" {type=Audio}
]
pages [Default {
  input_defaults {
   DefaultGainModifier=4
   Global-DebugMetronomeGain=0
   Global-Gain=1
   Global-Play=null
   Global-Stop=null
   Global-Tempo=120
   J0-FX0-IsActive=null
   J0-FX0-Param=3
   J0-Fusion=null
   J0-Gain=1
   J0-LoopInstanceId=-1
   J0-MarkupMidiTrackIndex=-1
   J0-Midi=null
   J0-Playing=null
   J0-SkipQuantization=null
   J0-Transpose=0
   J1-Fusion=null
   J1-Gain=1
   J1-LoopInstanceId=-1
   J1-MarkupMidiTrackIndex=-1
   J1-Midi=null
   J1-Playing=null
   J1-SkipQuantization=null
   J1-Transpose=0
   J2-Fusion=null
   J2-Gain=1
   J2-LoopInstanceId=-1
   J2-MarkupMidiTrackIndex=-1
   J2-Midi=null
   J2-Playing=null
   J2-SkipQuantization=null
   J2-Transpose=0
   J3-Fusion=null
   J3-Gain=1
   J3-LoopInstanceId=-1
   J3-MarkupMidiTrackIndex=-1
   J3-Midi=null
   J3-Playing=null
   J3-SkipQuantization=null
   J3-Transpose=0
   Multithreaded Rendering="<MetaSoundBuilderResult.SUCCEEDED: 0>"
   ReactivityUpdatePeriod=0.0333
   Sparks.MusicTimeNetSync.Seek=null
   Sparks.MusicTimeNetSync.Seek Target=null
   Sparks.MusicTimeNetSync.Speed=1
   UE.Source.OnPlay=null
   Crossover Frequencies [200,700,1200]
  }
  nodes [
   n1 = HarmonixNodes.TriggerToTransport(Play=^Global-Play, Stop=^Global-Stop, Trigger Seek=^Sparks.MusicTimeNetSync.Seek, Seek Target=^Sparks.MusicTimeNetSync.Seek Target)
   HarmonixNodesMetronome_0 = HarmonixNodes.Metronome(Tempo=^Global-Tempo, Transport=@n1.Transport, Speed=UE.InterpTo:Audio(Target=^Sparks.MusicTimeNetSync.Speed).Value) {also {defaults {"Loop Length (Bars)"=32, Pre-roll Bars=64}}}
   n2 = UE.Reroute(Value=@n1.Transport)
   HarmonixNodesMIDIPlayer_0 = HarmonixNodes.MIDIPlayer(MIDI Clock=@HarmonixNodesMetronome_0.Midi Clock, Transport=@n2.Value) {also {defaults {MIDI File=&/Empty32Bars, Simple Loop="<MetaSoundBuilderResult.SUCCEEDED: 0>"}}}
   n3 = UE.Reroute(Value=@HarmonixNodesMIDIPlayer_0.MIDI Clock)
   n4 = UE.Reroute(Value=@n2.Value)
   n5 = UE.Reroute(Value=@n3.Value)
   MSP_JamMusicPlayer_Internal_0 = &/MSP_JamMusicPlayer_Internal(
    Crossover Frequencies=^Crossover Frequencies
    J0-Fusion=^J0-Fusion
    J0-Gain=^J0-Gain
    J0-LoopInstanceId=^J0-LoopInstanceId
    J0-MarkupMidiTrackIndex=^J0-MarkupMidiTrackIndex
    J0-Midi=^J0-Midi
    J0-Playing=^J0-Playing
    J0-SkipQuantization=^J0-SkipQuantization
    J0-Transpose=^J0-Transpose
    J1-Fusion=^J1-Fusion
    J1-Gain=^J1-Gain
    J1-LoopInstanceId=^J1-LoopInstanceId
    J1-MarkupMidiTrackIndex=^J1-MarkupMidiTrackIndex
    J1-Midi=^J1-Midi
    J1-Playing=^J1-Playing
    J1-SkipQuantization=^J1-SkipQuantization
    J1-Transpose=^J1-Transpose
    J2-Fusion=^J2-Fusion
    J2-Gain=^J2-Gain
    J2-LoopInstanceId=^J2-LoopInstanceId
    J2-MarkupMidiTrackIndex=^J2-MarkupMidiTrackIndex
    J2-Midi=^J2-Midi
    J2-Playing=^J2-Playing
    J2-SkipQuantization=^J2-SkipQuantization
    J2-Transpose=^J2-Transpose
    J3-Fusion=^J3-Fusion
    J3-Gain=^J3-Gain
    J3-LoopInstanceId=^J3-LoopInstanceId
    J3-MarkupMidiTrackIndex=^J3-MarkupMidiTrackIndex
    J3-Midi=^J3-Midi
    J3-Playing=^J3-Playing
    J3-SkipQuantization=^J3-SkipQuantization
    J3-Transpose=^J3-Transpose
    Multithreaded Rendering=^Multithreaded Rendering
    ReactivityUpdatePeriod=^ReactivityUpdatePeriod
    Global-Gain=UE.InterpTo:Audio(Target=^Global-Gain).Value
    TransportIn=UE.Reroute(Value=UE.Reroute(Value=@n4.Value).Value).Value
    MidiClockIn=UE.Reroute(Value=UE.Reroute(Value=@n5.Value).Value).Value
   )
   n6 = /FMJamSystem/Audio/JamEffects/MSS_Jam_Effect_StutteRecapture(
    IsActive=^J0-FX0-IsActive
    Rate=^J0-FX0-Param
    MIDI Clock=UE.Reroute(Value=UE.Reroute(Value=@n3.Value).Value).Value
    Input L=UE.Reroute(Value=@MSP_JamMusicPlayer_Internal_0.Out Left).Value
    Input R=UE.Reroute(Value=@MSP_JamMusicPlayer_Internal_0.Out Right).Value
   )
   n7 = UE.Reroute(Value=UE.InterpTo:Audio(Target=^DefaultGainModifier).Value)
   Multiply_0 = UE.Multiply:Audio by Float(PrimaryOperand=@n6."UE.OutputFormat.Stereo.Audio:0", AdditionalOperands=@n7.Value)
   Multiply_1 = UE.Multiply:Audio by Float(PrimaryOperand=@n6."UE.OutputFormat.Stereo.Audio:1", AdditionalOperands=@n7.Value)
  ]
  links [
   {dst="@Out Looped Midi Clock", src=HarmonixNodesMIDIPlayer_0.MIDI Clock}
   {dst="@ClockOut", src=HarmonixNodesMetronome_0.Midi Clock}
   {dst="@AmplitudeOut", src=MSP_JamMusicPlayer_Internal_0.AmplitudeOut}
   {dst="@J0-Amplitude", src=MSP_JamMusicPlayer_Internal_0.J0-Amplitude}
   {dst="@J0-Band1", src=MSP_JamMusicPlayer_Internal_0.J0-Band1}
   {dst="@J0-Band2", src=MSP_JamMusicPlayer_Internal_0.J0-Band2}
   {dst="@J0-Band3", src=MSP_JamMusicPlayer_Internal_0.J0-Band3}
   {dst="@J0-Band4", src=MSP_JamMusicPlayer_Internal_0.J0-Band4}
   {dst="@J1-Amplitude", src=MSP_JamMusicPlayer_Internal_0.J1-Amplitude}
   {dst="@J1-Band1", src=MSP_JamMusicPlayer_Internal_0.J1-Band1}
   {dst="@J1-Band2", src=MSP_JamMusicPlayer_Internal_0.J1-Band2}
   {dst="@J1-Band3", src=MSP_JamMusicPlayer_Internal_0.J1-Band3}
   {dst="@J1-Band4", src=MSP_JamMusicPlayer_Internal_0.J1-Band4}
   {dst="@J2-Amplitude", src=MSP_JamMusicPlayer_Internal_0.J2-Amplitude}
   {dst="@J2-Band1", src=MSP_JamMusicPlayer_Internal_0.J2-Band1}
   {dst="@J2-Band2", src=MSP_JamMusicPlayer_Internal_0.J2-Band2}
   {dst="@J2-Band3", src=MSP_JamMusicPlayer_Internal_0.J2-Band3}
   {dst="@J2-Band4", src=MSP_JamMusicPlayer_Internal_0.J2-Band4}
   {dst="@J3-Amplitude", src=MSP_JamMusicPlayer_Internal_0.J3-Amplitude}
   {dst="@J3-Band1", src=MSP_JamMusicPlayer_Internal_0.J3-Band1}
   {dst="@J3-Band2", src=MSP_JamMusicPlayer_Internal_0.J3-Band2}
   {dst="@J3-Band3", src=MSP_JamMusicPlayer_Internal_0.J3-Band3}
   {dst="@J3-Band4", src=MSP_JamMusicPlayer_Internal_0.J3-Band4}
   {dst="@UE.OutputFormat.Stereo.Audio:0", src=Multiply_0.Out}
   {dst="@UE.OutputFormat.Stereo.Audio:1", src=Multiply_1.Out}
  ]
 }]
properties {
 assetImportData=&self:AssetImportData
 bEnableBaseSubmix=false
 bLooping=true
 duration=10000
 numChannels=2
 outputFormat=Stereo
 sampleRate=48000
 soundClassObject=&/Jam_Main
 virtualizationMode=PlayWhenSilent
 blockRateOverride {default=100}
 preEffectBusSends [{audioBus=&/AB_JamMusicBus, maxSendDistance=1000, minSendDistance=100, minSendLevel=0, sendLevel=1}]
 soundSubmixSends [{}]
}
