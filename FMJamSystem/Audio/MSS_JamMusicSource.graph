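# Stereo MetaSound source that plays the Jam music audio bus as an attenuated, looping sound.
# An Audio Bus Reader feeds both channels through a shared JamSourceGain multiply into a stereo mixer.
# Mixer gain is one minus the Cinematic_IsActive modulator value, so the music drops out as that control bus rises.
# It uses the Jam_Main sound class, keeps playing when silent, and takes volume modulation from the FortJam volume control bus.
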MetaSoundSource /FMJamSystem/Audio/MSS_JamMusicSource
inputs [
 JamSourceGain {type=Float}
 UE.Source.OnPlay {type=Trigger}
]
interfaces [UE.OutputFormat.Stereo:1.0, UE.Source:1.1]
outputs [
 "UE.OutputFormat.Stereo.Audio:0" {type=Audio}
 "UE.OutputFormat.Stereo.Audio:1" {type=Audio}
]
pages [Default {
  input_defaults {JamSourceGain=1.8, UE.Source.OnPlay=null}
  nodes [
   AudioBusReader2_0 = "UE.Audio Bus Reader (2)"() {also {defaults {Audio Bus=&/AB_JamMusicBus}}}
   AudioMixerAudioMixerStereo2_0 = "AudioMixer.Audio Mixer (Stereo, 2)"(
    In 0 R=UE.Multiply:Audio by Float(AdditionalOperands=^JamSourceGain, PrimaryOperand=@AudioBusReader2_0.Out 1).Out
    In 0 L=UE.Multiply:Audio by Float(AdditionalOperands=^JamSourceGain, PrimaryOperand=@AudioBusReader2_0.Out 0).Out
    Gain 0=UE.Subtract:Float(AdditionalOperands=Modulation.GetModulatorValue() {also {defaults {Modulator=/Game/Sounds/Modulation/Buses/CB_Cinematic_IsActive}}}.Out) {also {defaults {PrimaryOperand=1}}}.Out
   )
  ]
  links [
   {dst="@UE.OutputFormat.Stereo.Audio:0", src=AudioMixerAudioMixerStereo2_0.Out L}
   {dst="@UE.OutputFormat.Stereo.Audio:1", src=AudioMixerAudioMixerStereo2_0.Out R}
  ]
 }]
properties {
 assetImportData=&self:AssetImportData
 attenuationSettings=&/JamLoop_FreeAtten
 bLooping=true
 duration=10000
 maxDistance=1035
 numChannels=2
 outputFormat=Stereo
 sampleRate=48000
 soundClassObject=&/Jam_Main
 virtualizationMode=PlayWhenSilent
 volume=0.8
 blockRateOverride {default=100}
 modulationSettings {volumeRouting=Union, volumeModulationDestination {bEnableModulation=true, modulators [&/CB_FortJam_Volume]}}
}
