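# Sound cue for a looping metal ambient bed in the FMJamSystem audio set, routed to the Jam_Main sound class.
# An Enveloper wraps a looping wave player and fades the volume in from silence over the first two seconds.
# Playback runs at reduced volume with a shared attenuation asset and a concurrency override that caps simultaneous instances at two.
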SoundCue /FMJamSystem/Audio/SC_JamLoop_Ambient_Loop
root=Enveloper_0
nodes [
 {
  id=Enveloper_0
  children [WavePlayer_4]
  properties {pitchCurve {editorCurveData {keys [{time=0}]}}, volumeCurve {editorCurveData {keys [<at 0 0>,<at 2 1>]}}}
 }
 {id=WavePlayer_4, properties {bLooping=true, soundWaveAssetPtr=&/JamLoop_AmbientMetal_Loop}}
]
properties {
 attenuationSettings=&/JamLoop_FreeAmbLoopAtten
 bOverrideConcurrency=true
 maxDistance=2500
 soundClassObject=&/Jam_Main
 volumeMultiplier=0.5
 concurrencyOverrides {maxCount=2}
}
