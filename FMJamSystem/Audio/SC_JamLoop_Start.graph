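# SC_JamLoop_Start is a Sound Cue in the FMJamSystem audio that plays a crowd sound, apparently for the start of a jam loop.
# A Random node picks among three equally weighted crowd wave variations and is set to be excluded from branch culling.
# It uses the Jam_Main sound class and its own attenuation asset.
# A concurrency override allows one instance at a time, stopping the farthest and then preventing new ones.
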SoundCue /FMJamSystem/Audio/SC_JamLoop_Start
root=Random_0
nodes [
 {id=Random_0, children [WavePlayer_0,WavePlayer_1,WavePlayer_2], properties {bShouldExcludeFromBranchCulling=true, weights [1,1,1]}}
 {id=WavePlayer_0, properties {soundWaveAssetPtr=&/JamLoop_Start_Crowd_01}}
 {id=WavePlayer_1, properties {soundWaveAssetPtr=&/JamLoop_Start_Crowd_02}}
 {id=WavePlayer_2, properties {soundWaveAssetPtr=&/JamLoop_Start_Crowd_03}}
]
properties {
 attenuationSettings=&/JamLoop_StartAtten
 bOverrideConcurrency=true
 maxDistance=1000
 soundClassObject=&/Jam_Main
 volumeMultiplier=0.85
 concurrencyOverrides {maxCount=1, resolutionRule=StopFarthestThenPreventNew}
}
