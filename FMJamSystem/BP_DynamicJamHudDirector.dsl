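; Dynamic UI director for the FMJam HUD that shows the start-jam prompt when the playlist carries the FMJam.JamPrompts.AlwaysShow tag.
; On BeginPlay, cosmetic-capable instances await playlist data, add the prompt scene for the owning local player, then update input aliases.
; The update clears tracked aliases, rebuilding a "Play Loop (Hold)" emote-picker alias while the prompt flag is set.
; EndPlay clears that flag and aliases and cancels any playspace tag listeners.
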(event EventBeginPlay
  (bind _returnvalue (DynamicUIDirector|GetOwningLocalPlayerController))
  (|CallParentFunction)
  ;[1 1 1] Add the jam prompts scene if the playlist wants it always on
  (Networking|CanExecuteCosmeticEvents
    (:True
      (bind (_gamestate _playlist _playlistcontexttags) (Class|FortAsyncActionCurrentPlaylistReady|PlaylistDataReadyAsync)
        (:OnReady
          (if (GameplayTags|HasTag _playlistcontexttags "(TagName=\"FMJam.JamPrompts.AlwaysShow\")")
            (Utilities|IsValid _returnvalue
              (:Is%20Valid
                (DynamicUI|AddScene (WorldSubsystems|GetDynamicUIManager) (Variables|Default|GetStartJamPromptScene) _returnvalue)
                (Variables|Default|SetbHaveAddedStartJamPrompt true)
                (CallFunction|UpdateInputAliases)))
            (else
              _)))))))

(event Custom|Update%20Input%20Aliases
  (bind _returnvalue (DynamicUIDirector|GetOwningLocalPlayerController))
  ;[0.0323 0.15 0 0.5] Manage Input Aliases
  (if (Variables|Default|GetbHaveAddedStartJamPrompt)
    (Utilities|FlowControl|Sequence
      (:then_0
        _)
      (:then_1
        ;; Construct new Input Aliases as required
        (bind _returnvalue_2 (Game|ConstructObjectfromClass :Class "/Script/InputPromptManagerRuntime.InputAlias" :InputArchetype "/FMJamSystem/Input/Archetypes/InputArchetype_ContributeJam.InputArchetype_ContributeJam" :Priority 2000 :UISortOrder 1))
        (InputAbstraction|Aliases|LegacyInput|AssignLegacyBindingData _returnvalue_2 (struct :SpecializedActionName "GamepadEmotePickerOrReplayLastEmoteHold" :GeneralActionName "EmotePicker" :ActionIdentifierName "None" :ActionDescription "NSLOCTEXT(\"[805C00A3BFAA66136C7EF94AE7078A63]\", \"95026CA64BDA387F558C14845450389F\", \"Play Loop (Hold)\")"))
        (Utilities|Array|Add (Variables|Default|GetActive%20Input%20Aliases) _returnvalue_2))
      (:then_2
        ;; Add the new input aliases
        (bind _returnvalue_3 (ManagedInputPrompts|GetInputPromptManager _returnvalue))
        (ManagedInputPrompts|AddInputAliases _returnvalue_3 (Variables|Default|GetActive%20Input%20Aliases)))))
  ;; Remove input aliases and clear our internal tracking array
  (bind _returnvalue_4 (ManagedInputPrompts|GetInputPromptManager _returnvalue))
  (ManagedInputPrompts|RemoveInputAliases _returnvalue_4 (Variables|Default|GetActive%20Input%20Aliases))
  (Utilities|Array|Clear (Variables|Default|GetActive%20Input%20Aliases)))

(event EventEndPlay
  (params (EndPlayReason :type EEndPlayReason%20Enum :default "Destroyed"))
  (bind _removeshowwheninplayspacetaglistener (Variables|Default|GetRemoveShowWhenInPlayspaceTagListener))
  (bind _addshowwheninplayspacetaglistener (Variables|Default|GetAddShowWhenInPlayspaceTagListener))
  (Utilities|FlowControl|Sequence
    (:then_0
      ; Cleanup Input Aliases
      (Variables|Default|SetbHaveAddedStartJamPrompt false)
      (CallFunction|UpdateInputAliases))
    (:then_1
      ;[0.5284 1 0.6603] Cleanup Tag listeners
      (Utilities|IsValid _removeshowwheninplayspacetaglistener
        (:Is%20Valid
          (AsyncAction|Cancel _removeshowwheninplayspacetaglistener)
          (Variables|Default|SetAddShowWhenInPlayspaceTagListener 0))))
    (:then_2
      (Utilities|IsValid _addshowwheninplayspacetaglistener
        (:Is%20Valid
          (AsyncAction|Cancel _addshowwheninplayspacetaglistener)
          (Variables|Default|SetRemoveShowWhenInPlayspaceTagListener 0))))
    (:then_3
      (|CallParentFunction "Destroyed"))))

(fn ConstructionScript)
