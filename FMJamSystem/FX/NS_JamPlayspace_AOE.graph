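# Niagara effect that draws a glowing cylinder area marker for a Jam playspace, sized by user radius and height scales.
# Two near-identical mesh emitters choose three-color palettes from arrays by System.ColorInt, differing slightly in tint and contested opacity.
# User.IsContested switches spawn rate, alpha curve and a dynamic material parameter curve.
# That parameter reverses direction between states, scaled by separate ring speed variables, apparently driving moving rings.
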NiagaraSystem /FMJamSystem/FX/NS_JamPlayspace_AOE
dependencies {
 usedDataInterfaces [&NiagaraDataInterfaceArrayColor, &NiagaraDataInterfaceCurve]
 usedDynamicInputs []
 usedModules [
  &SystemState
  &self:SystemSpawnScript.SystemScriptSource.SystemScriptGraph.NiagaraNodeAssignment_0.SetVariables_#1
 ]
}
emitters {
 CylinderGlow {
  emitterUpdateScript [
   &EmitterState(Life Cycle Mode=::NewEnumerator1, Loop Duration=1, UseLoopDelay=false)
   &SpawnRate(
    SpawnRate=&MakeCustomFloatFromBool(Bool=$User.IsContested, True Float=1.3, False Float=0.5)
    Use Spawn Probability=false
   )
   /SetVariables(
    Emitter.BrightColor:LinearColor=&SelectLinearColorFromArray(
     Array Sampling Mode=::NewEnumerator1
     Direct Array Index=$System.ColorInt
     Color Selection Array=DataInterface() {propertyValues {ColorData [<rgba 0.5244 0.7123 0.8871 0>, <rgba 2 0.3333 0.0135 0>, <rgba 0.1497 0.8948 2 0>]}}
     Random Seed=?
    )
    Emitter.HotBaseColor:LinearColor=&SelectLinearColorFromArray(
     Array Sampling Mode=::NewEnumerator1
     Direct Array Index=$System.ColorInt
     Color Selection Array=DataInterface() {propertyValues {ColorData [<rgba 4.6733 5 4.7746 0>, <rgba 5 1.3178 0.1619 0>, <rgba 0.3906 3.4901 5 0>]}}
     Random Seed=?
    )
    Emitter.TopColor:LinearColor=&SelectLinearColorFromArray(
     Array Sampling Mode=::NewEnumerator1
     Direct Array Index=$System.ColorInt
     Color Selection Array=DataInterface() {propertyValues {ColorData [<rgba 0.5461 0.5586 0.725 0>, <rgba 1 0.0708 0.1305 0>, <rgba 0.1062 0.2026 1 0>]}}
     Random Seed=?
    )
   ) {isSetParameters=true, module=SetVariables_#2}
  ]
  particleSpawnScript [&InitializeParticle(
    Color Mode=::NewEnumerator0
    Lifetime=4
    Mesh Scale=&MakeVector(
     X=&Multiply_Float(A=8.3, B=$User.RadiusScale)
     Y=&Multiply_Float(A=8.3, B=$User.RadiusScale)
     Z=&Multiply_Float(A=2, B=$User.HeightScale)
    )
    Mesh Scale Mode=::NewEnumerator4
    UsePositionOffset=false
    Write Material Random=true
   )]
  particleUpdateScript [
   &ParticleState(
    Kill Particles When Lifetime Has Elapsed=false
    Let Infinitely Lived Particles Die When Emitter Deactivates=true
    Loop Particles Lifetime=true
   )
   &ScaleColor(
    Scale Alpha=&MakeCustomFloatFromBool(
     Bool=$User.IsContested
     True Float=&FloatFromCurve(FloatCurve=DataInterface() {propertyValues {Curve {keys [
          {interpMode=RCIM_Cubic, tangentMode=RCTM_User, time=0, value=0}
          {arriveTangent=4.0808, interpMode=RCIM_Cubic, tangentMode=RCTM_Break, time=1, value=1}
         ]}}})
     False Float=&FloatFromCurve(FloatCurve=DataInterface() {propertyValues {Curve {keys [
          {interpMode=RCIM_Cubic, leaveTangent=-3.4875, tangentMode=RCTM_Break, time=0, value=1}
          {interpMode=RCIM_Cubic, tangentMode=RCTM_User, time=1, value=0}
         ]}}})
    )
    Scale RGB=&VectorFromFloat(Value=1)
    ScaleA=true
    ScaleRGB=true
   ) {enabled=false, module=ScaleColor002}
   &ScaleColor(
    Scale Alpha=&MakeCustomFloatFromBool(
     Bool=$User.IsContested
     True Float=&FloatFromCurve(FloatCurve=DataInterface() {propertyValues {Curve {keys [
          {interpMode=RCIM_Cubic, tangentMode=RCTM_User, time=0, value=0}
          {arriveTangent=5.1713, interpMode=RCIM_Cubic, tangentMode=RCTM_Break, time=0.9, value=1}
          {arriveTangent=4.0808, interpMode=RCIM_Cubic, tangentMode=RCTM_Break, time=1, value=0}
         ]}}})
     False Float=&FloatFromCurve(FloatCurve=DataInterface() {propertyValues {Curve {keys [
          {interpMode=RCIM_Cubic, leaveTangent=-3.4875, tangentMode=RCTM_Break, time=0, value=1}
          {interpMode=RCIM_Cubic, tangentMode=RCTM_User, time=1, value=0}
         ]}}})
    )
    Scale RGB=&VectorFromFloat(Value=1)
    ScaleA=true
    ScaleRGB=true
   )
   &DynamicMaterialParameters(
    Index 0 Param 1=&MakeCustomFloatFromBool(
     Bool=$User.IsContested
     True Float=&FloatFromCurve(
      FloatCurve=DataInterface() {propertyValues {Curve {keys [{interpMode=RCIM_Cubic, tangentMode=RCTM_User, time=0, value=1}, {interpMode=RCIM_Cubic, tangentMode=RCTM_User, time=1, value=0}]}}}
      Scale Curve=$User.RingSpeedContested
     )
     False Float=&FloatFromCurve(
      FloatCurve=DataInterface() {propertyValues {Curve {keys [{interpMode=RCIM_Cubic, tangentMode=RCTM_User, time=0, value=0}, {interpMode=RCIM_Cubic, tangentMode=RCTM_User, time=1, value=0.4}]}}}
      Scale Curve=$User.RingSpeedUncontested
     )
    )
    Module.Param0WriteEnabled=true
    Module.Param1WriteEnabled=false
    Module.Param2WriteEnabled=false
    Module.Param3WriteEnabled=false
   ) {module=DynamicMaterialParameters001}
  ]
  rendererClasses [&NiagaraMeshRendererProperties]
  renderers [{rendererClass=&NiagaraMeshRendererProperties, rendererIndex=0}]
 }
 CylinderGlow001 {
  emitterUpdateScript [
   &EmitterState(Life Cycle Mode=::NewEnumerator1, Loop Duration=1, UseLoopDelay=false)
   &SpawnRate(
    SpawnRate=&MakeCustomFloatFromBool(Bool=$User.IsContested, True Float=1.3, False Float=0.5)
    Use Spawn Probability=false
   )
   /SetVariables(
    Emitter.BrightColor:LinearColor=&SelectLinearColorFromArray(
     Array Sampling Mode=::NewEnumerator1
     Direct Array Index=$System.ColorInt
     Color Selection Array=DataInterface() {propertyValues {ColorData [<rgba 0.8871 0.8871 0.8871 0>, <rgba 2 0.4 0 0>, <rgba 0.1497 0.8948 2 0>]}}
     Random Seed=?
    )
    Emitter.HotBaseColor:LinearColor=&SelectLinearColorFromArray(
     Array Sampling Mode=::NewEnumerator1
     Direct Array Index=$System.ColorInt
     Color Selection Array=DataInterface() {propertyValues {ColorData [<rgba 5 5 5 0>, <rgba 5 1 0 0>, <rgba 0.3906 3.4901 5 0>]}}
     Random Seed=?
    )
    Emitter.TopColor:LinearColor=&SelectLinearColorFromArray(
     Array Sampling Mode=::NewEnumerator1
     Direct Array Index=$System.ColorInt
     Color Selection Array=DataInterface() {propertyValues {ColorData [<rgba 0.5461 0.5586 0.725 0>, <rgba 1 0 0.0115 0>, <rgba 0.1062 0.2026 1 0>]}}
     Random Seed=?
    )
   ) {isSetParameters=true, module=SetVariables_#2}
  ]
  particleSpawnScript [&InitializeParticle(
    Color Mode=::NewEnumerator0
    Lifetime=4
    Mesh Scale=&MakeVector(
     X=&Multiply_Float(A=8.3, B=$User.RadiusScale)
     Y=&Multiply_Float(A=8.3, B=$User.RadiusScale)
     Z=&Multiply_Float(A=2, B=$User.HeightScale)
    )
    Mesh Scale Mode=::NewEnumerator4
    UsePositionOffset=false
    Write Material Random=true
   )]
  particleUpdateScript [
   &ParticleState(
    Kill Particles When Lifetime Has Elapsed=false
    Let Infinitely Lived Particles Die When Emitter Deactivates=true
    Loop Particles Lifetime=true
   )
   &ScaleColor(
    Scale Alpha=&MakeCustomFloatFromBool(
     Bool=$User.IsContested
     True Float=&FloatFromCurve(FloatCurve=DataInterface() {propertyValues {Curve {keys [
          {interpMode=RCIM_Cubic, tangentMode=RCTM_User, time=0, value=0}
          {arriveTangent=4.0808, interpMode=RCIM_Cubic, tangentMode=RCTM_Break, time=1, value=1}
         ]}}})
     False Float=&FloatFromCurve(FloatCurve=DataInterface() {propertyValues {Curve {keys [
          {interpMode=RCIM_Cubic, leaveTangent=-3.4875, tangentMode=RCTM_Break, time=0, value=1}
          {interpMode=RCIM_Cubic, tangentMode=RCTM_User, time=1, value=0}
         ]}}})
    )
    Scale RGB=&VectorFromFloat(Value=1)
    ScaleA=true
    ScaleRGB=true
   ) {enabled=false, module=ScaleColor002}
   &ScaleColor(
    Scale Alpha=&MakeCustomFloatFromBool(
     Bool=$User.IsContested
     True Float=&FloatFromCurve(
      FloatCurve=DataInterface() {propertyValues {Curve {keys [
          {interpMode=RCIM_Cubic, tangentMode=RCTM_User, time=0, value=0}
          {arriveTangent=5.1713, interpMode=RCIM_Cubic, tangentMode=RCTM_Break, time=0.9, value=1}
          {arriveTangent=4.0808, interpMode=RCIM_Cubic, tangentMode=RCTM_Break, time=1, value=0}
         ]}}}
      Scale Curve=0.5
     )
     False Float=&FloatFromCurve(FloatCurve=DataInterface() {propertyValues {Curve {keys [
          {interpMode=RCIM_Cubic, leaveTangent=-3.4875, tangentMode=RCTM_Break, time=0, value=1}
          {interpMode=RCIM_Cubic, tangentMode=RCTM_User, time=1, value=0}
         ]}}})
    )
    Scale RGB=&VectorFromFloat(Value=1)
    ScaleA=true
    ScaleRGB=true
   )
   &DynamicMaterialParameters(
    Index 0 Param 1=&MakeCustomFloatFromBool(
     Bool=$User.IsContested
     True Float=&FloatFromCurve(
      FloatCurve=DataInterface() {propertyValues {Curve {keys [{interpMode=RCIM_Cubic, tangentMode=RCTM_User, time=0, value=1}, {interpMode=RCIM_Cubic, tangentMode=RCTM_User, time=1, value=0}]}}}
      Scale Curve=$User.RingSpeedContested
     )
     False Float=&FloatFromCurve(
      FloatCurve=DataInterface() {propertyValues {Curve {keys [{interpMode=RCIM_Cubic, tangentMode=RCTM_User, time=0, value=0}, {interpMode=RCIM_Cubic, tangentMode=RCTM_User, time=1, value=0.4}]}}}
      Scale Curve=$User.RingSpeedUncontested
     )
    )
    Module.Param0WriteEnabled=true
    Module.Param1WriteEnabled=false
    Module.Param2WriteEnabled=false
    Module.Param3WriteEnabled=false
   ) {module=DynamicMaterialParameters001}
  ]
  rendererClasses [&NiagaraMeshRendererProperties]
  renderers [{rendererClass=&NiagaraMeshRendererProperties, rendererIndex=0}]
 }
}
summary {
 systemName=NS_JamPlayspace_AOE
 emitters [
  {emitterName=CylinderGlow, rendererClasses [&NiagaraMeshRendererProperties]}
  {emitterName=CylinderGlow001, rendererClasses [&NiagaraMeshRendererProperties]}
 ]
 userVariables [
  User.Color {type=NiagaraInt32}
  User.HeightScale {type=NiagaraFloat, defaultValue=1, description=""}
  User.IsContested {type=NiagaraBool}
  User.RadiusScale {type=NiagaraFloat, defaultValue=1, description=""}
  User.RingSpeedContested {type=NiagaraFloat, defaultValue=444, description=""}
  User.RingSpeedUncontested {type=NiagaraFloat, defaultValue=444, description=""}
 ]
}
