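; Gameplay ability for the Jam loop emote, which plays a player's chosen song part in a shared jam session.
; The server joins or spawns a jam session or playspace, then after privacy, blocking, ring and capacity checks asks the music manager to play the loop.
; Gameplay tag listeners and loop router events drive setup and end the ability.
; Montage starts are watched to mimic emote cancellation, and ending stops the owned loop.
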(event Ability|EventActivateAbility
  (bind _playerpawn (Variables|Default|GetPlayerPawn))
  (bind _returnvalue (Fortnite|GetFortPlayerControllerfromActor _playerpawn))
  (bind _returnvalue_2 (Class|FortPlayerController|GetContextTrackerComponent _returnvalue))
  ;[0 0.3855 1] Ability started
  (|CallParentFunction)
  (Ability|HasAuthority
    (:True
      ;;[1 1 1] Figure out if this was activated by QuickJoin.  This is only valid on the server
      (Variables|Default|SetActivationFromQuickPlay (GameplayTags|HasMatchingGameplayTag (Ability|GetAbilitySystemComponent (Variables|Default|GetPlayerPawn)) (Class|JamPlayParamsUtilities|GetJamStartedbyQuickJoinTag))))
    (:False
      _))
  ;;[1 1 1] FORT-719232 - The Parent ActivateAbility could trigger events and side effects that cause the ability to end by the time it returns. Try to catch that here and avoid crashing.
  (if (Variables|Default|GetAbilityEnded)
    (Fortnite|FortShippingLog self "GAB_Emote_JamLoop already ended during parent ActivateAbility" true)
    (else
      (Utilities|IsValid self
        (:Is%20Valid
          (CallFunction|ResetPropertiesonActivation)
          (Fortnite|FortShippingLog self "Starting GAB_Emote_JamLoop")
          (CallFunction|TrySimulateMontageDrivenBehavior)
          (Ability|HasAuthority
            (:True
              (Utilities|FlowControl|Sequence
                (:then_0
                  ;;[1 1 1] Create the Playspace if needed on server
                  (CallFunction|JoinorCreatePlayspace))
                (:then_1
                  ;;[1 1 1] Wait for the participant tag to be added so that we know we can find the Playspace
                  (bind _asynctaskproxy (Ability|Tasks|WaitGameplayTagAddToActor (Variables|/Script/Engine.Pawn|GetPlayerState (Variables|Default|GetPlayerPawn)) "(TagName=\"FMJam.Session.Member.Server\")")
                    (:then
                      (Variables|Default|SetParticipantTagAddedListenerHandle _asynctaskproxy))
                    (:Added
                      (Utilities|FlowControl|Sequence
                        (:then_0
                          (CallFunction|OnParticipantTagAdded))
                        (:then_1
                          ;;[1 1 1] Listen for all participant tags being removed (only on the server)
                          (bind _asynctaskproxy_2 (Ability|Tasks|WaitGameplayTagRemoveFromActor (Variables|/Script/Engine.Pawn|GetPlayerState (Variables|Default|GetPlayerPawn)) "(TagName=\"FMJam.Session.Member\")")
                            (:then
                              (Variables|Default|SetParticipantTagRemovedListenerHandle _asynctaskproxy_2))
                            (:Removed
                              (CallFunction|OnParticipantTagRemovedServer))))))))
                (:then_2
                  ;;[1 1 1] Listen for the jam blocking tag
                  (bind _asynctaskproxy_3 (Ability|Tasks|WaitGameplayTagAddToActor (Variables|/Script/Engine.Pawn|GetPlayerState (Variables|Default|GetPlayerPawn)) "(TagName=\"FMJam.BlockJamEmotes\")")
                    (:then
                      (Variables|Default|SetBlockJamEmotesTagAddedListenerHandle _asynctaskproxy_3))
                    (:Added
                      (CallFunction|OnBlockJamEmotesTagAddedServer))))))
            (:False
              ;;[1 1 1] Listen separately for the client participant added because if we listen for the server or generic one, we fire before we are actually added to the playspace (on the client)
              (bind _asynctaskproxy_4 (Ability|Tasks|WaitGameplayTagAddToActor (Variables|/Script/Engine.Pawn|GetPlayerState (Variables|Default|GetPlayerPawn)) "(TagName=\"FMJam.Session.Member.Client\")")
                (:then
                  (Variables|Default|SetParticipantTagAddedListenerHandle _asynctaskproxy_4)
                  ;;[1 1 1] End Jam traversal if we auto-run out of it on mobile
                  (Movement|BindEventtoOnAutoRunEnabled (Fortnite|GetFortPlayerControllerfromActor (Variables|Default|GetPlayerPawn)) (EventDispatchers|CreateEvent :!event Custom|HandleOnAutoRunEnabled))
                  ;;[1 1 1] Update context for HUD Layout
                  (Utilities|IsValid _returnvalue
                    (:Is%20Valid
                      (Utilities|IsValid _returnvalue_2
                        (:Is%20Valid
                          (Class|FortControllerComponentInputContextTracker|UpdateContexts _returnvalue_2 (Variables|Default|GetContextTags) (Variables|Default|GetEmpty%20Context%20Tags) true)
                          ;[1 0 0.3813] Debug logs for FORT-736709
                          (Fortnite|FortShippingLog self (Utilities|String|Append "[GAB_Emote_JamLoop - ActivateAbility] Context Tags To Add: [" (GameplayTags|GetDebugStringfromGameplayTagContainer (Variables|Default|GetContextTags)) "] - Active Context Tags: [" (GameplayTags|GetDebugStringfromGameplayTagContainer (Class|FortControllerComponentInputContextTracker|GetActiveContexts _returnvalue_2)) "]")))))))
                (:Added
                  (CallFunction|OnParticipantTagAdded))))))
        (:Is%20Not%20Valid
          (Fortnite|FortShippingLog self "GAB_Emote_JamLoop became invalid during parent ActivateAbility" true))))))

(event Custom|TrySimulateMontageDrivenBehavior
  ;[1 1 1] Stop Current Montage and Bind to Montage Started so we can emulate cancellation on montage start like Emotes. Only Relevant when we don't play a montage ourselves
  (Ability|StopCurrentMontage) ; Stop current montage to simulate montage interruption caused by playing another emote
  (EventDispatchers|AssignOnMontageStarted (Components|SkeletalMesh|GetAnimInstance (Variables|/Script/Engine.Character|GetMesh (Ability|GetActivatingPawn))) (EventDispatchers|CreateEvent :!event On%20Montage%20Started)))

(event Ability|EventOnEndAbility
  (params (bWasCancelled :type Boolean))
  (bind _playerpawn (Variables|Default|GetPlayerPawn))
  (bind _returnvalue (Fortnite|GetFortPlayerControllerfromActor _playerpawn))
  (bind _playerpawn_2 (Variables|Default|GetPlayerPawn))
  (bind _returnvalue_2 (Fortnite|GetFortPlayerControllerfromActor _playerpawn_2))
  (bind _returnvalue_3 (Class|FortPlayerController|GetContextTrackerComponent _returnvalue))
  (bind _jamloopstoppedasynctaskref (Variables|Default|GetJamLoopStoppedAsyncTaskRef))
  (bind _participanttagremovedlistenerhandle (Variables|Default|GetParticipantTagRemovedListenerHandle))
  (bind _jamloopstartedasynctaskref (Variables|Default|GetJamLoopStartedAsyncTaskRef))
  (bind _participanttagaddedlistenerhandle (Variables|Default|GetParticipantTagAddedListenerHandle))
  (bind _blockjamemotestagaddedlistenerhandle (Variables|Default|GetBlockJamEmotesTagAddedListenerHandle))
  ;[0 1 0.4643] The ability has ended - clean up!
  (|CallParentFunction bWasCancelled)
  (Variables|Default|SetAbilityEnded true)
  (Fortnite|FortShippingLog self "Ending Ability")
  (CallFunction|UnbindNotifierEvent)
  (Utilities|FlowControl|Sequence
    (:then_0
      (Ability|HasAuthority
        (:True
          (bind _musicmanager (Variables|Default|GetMusicManager)
            (:then
              (Utilities|IsValid (Mojo|JamSession|GetMusicSlotbyLoopInstanceId _musicmanager (Variables|Default|GetLoopInstanceId))
                (:Is%20Valid
                  (Mojo|JamSession|StopLoopforPlayer _musicmanager (Default|GetGameAccountId (Fortnite|GetFortPlayerControllerfromActor (Variables|Default|GetPlayerPawn))) true true))) ; if our loop is still in the mix, we're responsible for cleaning up.\notherwise, another ability has already taken ownership.
            )
            (:else)))))
    (:then_1
      (Utilities|IsValid _jamloopstartedasynctaskref
        (:Is%20Valid
          (GameplayEventRouter|StopListeningtoEvent _jamloopstartedasynctaskref)
          (Variables|Default|SetJamLoopStartedAsyncTaskRef 0))))
    (:then_2
      (Utilities|IsValid _jamloopstoppedasynctaskref
        (:Is%20Valid
          (GameplayEventRouter|StopListeningtoEvent _jamloopstoppedasynctaskref)
          (Variables|Default|SetJamLoopStoppedAsyncTaskRef 0))))
    (:then_3
      (Utilities|IsValid _participanttagaddedlistenerhandle
        (:Is%20Valid
          (AsyncAction|Cancel _participanttagaddedlistenerhandle)
          (Variables|Default|SetParticipantTagAddedListenerHandle 0))))
    (:then_4
      (Utilities|IsValid _participanttagremovedlistenerhandle
        (:Is%20Valid
          (AsyncAction|Cancel _participanttagremovedlistenerhandle)
          (Variables|Default|SetParticipantTagRemovedListenerHandle 0))))
    (:then_5
      (Utilities|IsValid _blockjamemotestagaddedlistenerhandle
        (:Is%20Valid
          (AsyncAction|Cancel _blockjamemotestagaddedlistenerhandle)
          (Variables|Default|SetBlockJamEmotesTagAddedListenerHandle 0))))
    (:then_6
      ;;[1 1 1] Update context for HUD Layout
      (Utilities|IsValid _returnvalue
        (:Is%20Valid
          (Utilities|IsValid _returnvalue_3
            (:Is%20Valid
              (Class|FortControllerComponentInputContextTracker|UpdateContexts _returnvalue_3 (Variables|Default|GetEmpty%20Context%20Tags) (Variables|Default|GetContextTags) true)
              ;[1 0 0.3813] Debug logs for FORT-736709
              (Fortnite|FortShippingLog self (Utilities|String|Append "[GAB_Emote_JamLoop - OnEndAbility] Context Tags To Remove: [" (GameplayTags|GetDebugStringfromGameplayTagContainer (Variables|Default|GetContextTags)) "] - Active Context Tags: [" (GameplayTags|GetDebugStringfromGameplayTagContainer (Class|FortControllerComponentInputContextTracker|GetActiveContexts _returnvalue_3)) "]")))))))
    (:then_7
      (Utilities|IsValid _returnvalue_2
        (:Is%20Valid
          (Movement|UnbindEventfromOnAutoRunEnabled _returnvalue_2 (EventDispatchers|CreateEvent :!event Custom|HandleOnAutoRunEnabled)))))))

(event Custom|BindPlayspaceEvents
  (bind _jamsession (Variables|Default|GetJamSession))
  (bind _asactor (Utilities|Casting|CastToActor _jamsession))
  ;[1 0.5575 0.142] Bind the events on the playspace
  (Utilities|FlowControl|Sequence
    (:then_0
      (Ability|HasAuthority
        (:True
          (Collision|BindEventtoOnActorEndOverlap _asactor (EventDispatchers|CreateEvent :!event Custom|ActorLeftPlayspace))
          (Game|BindEventtoOnEndPlay _asactor (EventDispatchers|CreateEvent :!event Custom|PlayspaceEndsplay)))))
    (:then_1
      (CallFunction|ListentoLoopStartedEvent))))

(event Custom|ActorLeftPlayspace
  (params (OverlappedActor :type Actor%20Object%20Reference) (OtherActor :type Actor%20Object%20Reference))
  ;[0.5473 1 0.8263] The player has left the playspace - stop the ability!
  (if (== OtherActor (Variables|Default|GetPlayerPawn))
    (Fortnite|FortShippingLog self "Actor left playspace. Ending Ability")
    (Ability|EndAbility)))

(event Custom|Handle%20Jam%20Error
  (params (Player%20Pawn :type Fort%20Player%20Pawn%20Object%20Reference) (Play%20Montage :type Boolean) (Failed%20Reason :type Gameplay%20Tag%20Container%20Structure) (ShouldEndAbility :type Boolean))
  ;[1 0.3633 0.3434] Error Handling
  (CallFunction|PlayJamErrorMessageClient Player%20Pawn Play%20Montage Failed%20Reason)
  (if ShouldEndAbility
    (Fortnite|FortShippingLog self (Utilities|String|Append "Handled Jam Error: " (GameplayTags|GetDebugStringfromGameplayTagContainer Failed%20Reason) ". Ending Ability"))
    (Ability|EndAbility)))

(event Custom|PlayJamErrorMessage_Client
  (params (Player%20Pawn :type Fort%20Player%20Pawn%20Object%20Reference) (Play%20Montage :type Boolean) (Failed%20Reason :type Gameplay%20Tag%20Container%20Structure))
  ;[1 0.3633 0.3434] Error Handling
  (FailedAction|UnabletoPerformAction Player%20Pawn Play%20Montage Failed%20Reason))

(event Custom|PlayspaceEndsplay
  (params (Actor :type Actor%20Object%20Reference) (EndPlayReason :type EEndPlayReason%20Enum :default "Destroyed"))
  ;[0.5473 1 0.8263] The playspace is gone - stop the ability!
  (Fortnite|FortShippingLog self "Playspace ending. Ending Ability")
  (Ability|EndAbility))

(event EventPlayInitialEmoteMontage
  (CallFunction|EnsureEmoteStartNotification))

(event Custom|Ensure%20Emote%20Start%20Notification
  (Utilities|FlowControl|Delay 0.2)
  (if (not (or (Variables|Default|GetAbilityEnded) (Variables|Default|GetbAbilityStopped)))
    (Emote|EmoteStarted (Ability|GetActivatingPawn) (CallFunction|GetCurrentEmoteItemDefinition))))

(event Custom|EventPlayMontage)

(event Custom|ListenToLoopStartedEvent
  (bind _playernetid (Variables|Default|GetPlayerNetId))
  ;[0.4957 0.5654 1] A Loop Started Event
  (bind (_asynctaskproxy _payload) (GameplayEventRouter|StartListeningToEvent (Class|JamSessionInterface|GetGameplayEventRouterComponent (Variables|Default|GetJamSession)) "/Script/FMJamSessionRuntime.JamEvent_JamLoopStarted")
    (:then
      (Variables|Default|SetJamLoopStartedAsyncTaskRef _asynctaskproxy)
      ;[1 1 1] If the Playspace thinks that we haven't played yet, but we have data, it must mean that we have missed the started event already. Freqeuntly happens on the first loop of a jam session, randomly after.
      (bind _musicmanager (Variables|Default|GetMusicManager)
        (:then
          (Utilities|IsValid (Mojo|JamSession|GetMusicSlotforPlayerId _musicmanager _playernetid)
            (:Is%20Valid
              (if (and (== (Default|GetSongShortName) (Class|JamMusicSlot|GetCurrentSongShortName (Mojo|JamSession|GetMusicSlotforPlayerId _musicmanager _playernetid))) (Utilities|Enum|Equal(Enum) (Class|JamMusicSlot|GetCurrentLoopType (Mojo|JamSession|GetMusicSlotforPlayerId _musicmanager _playernetid)) (CallFunction|GetLoopType)))
                (CallFunction|OnOwnJamLoopStartedEvent (Class|JamMusicSlot|GetCurrentLoopInstanceId (Mojo|JamSession|GetMusicSlotforPlayerId _musicmanager _playernetid)))))))
        (:else)))
    (:OnEventReceived
      ;;[1 1 1] Ignore autojammers, this ability only cares about the player
      (if (.bIsAutoJammer (Utilities|Struct|BreakJamEventJamLoopStarted _payload))
        _
        (elif (== (.PlayerNetId (Utilities|Struct|BreakJamEventJamLoopStarted _payload)) (Variables|Default|GetPlayerNetId))
          (CallFunction|OnOwnJamLoopStartedEvent (.LoopInstanceId (Utilities|Struct|BreakJamEventJamLoopStarted _payload))))))))

(event Custom|ListenToLoopEndedEvent
  ;[0.3281 1 0.9202] A Loop Stopped Event
  (bind (_asynctaskproxy _payload) (GameplayEventRouter|StartListeningToEvent (Class|JamSessionInterface|GetGameplayEventRouterComponent (Variables|Default|GetJamSession)) "/Script/FMJamSessionRuntime.JamEvent_JamLoopStopped")
    (:then
      (Variables|Default|SetJamLoopStoppedAsyncTaskRef _asynctaskproxy))
    (:OnEventReceived
      (if (== (.LoopInstanceId (Utilities|Struct|BreakJamEventJamLoopStopped _payload)) (Variables|Default|GetLoopInstanceId))
        (CallFunction|OnOwnJamLoopStoppedEvent)))))

(event Custom|OnOwnJamLoopStartedEvent
  (params (LoopInstanceId :type Integer))
  (bind _jamloopstartedasynctaskref (Variables|Default|GetJamLoopStartedAsyncTaskRef))
  ;[0.7333 0.4267 1] Track the loop we started so that we don't try to stop the wrong one later
  (Variables|Default|SetLoopInstanceId LoopInstanceId)
  (CallFunction|EventPlayMontage)
  (CallFunction|ListentoLoopEndedEvent)
  ;;[1 1 1] Stop listening for loop started, we have already handled our own loop
  (Utilities|IsValid _jamloopstartedasynctaskref
    (:Is%20Valid
      (GameplayEventRouter|StopListeningtoEvent _jamloopstartedasynctaskref)
      (Variables|Default|SetJamLoopStartedAsyncTaskRef 0))))

(event Custom|OnOwnJamLoopStoppedEvent
  ;[1 0.5575 0.142] Listen to events for loops in this playspace
  (Ability|HasAuthority
    (:True
      ;[0.3181 1 0.3683] Stop Jamming if the player's loop has been stopped
      (Fortnite|FortShippingLog self "Own loop ended. Ending Ability")
      (Ability|EndAbility))))

(event Custom|OnParticipantTagAdded
  (bind _returnvalue (Ability|GetActivatingPawn))
  (bind (_outfoundcomponent _returnvalue_2) (Class|JamSessionComponentLibrary|GetComponentOnJamSession "/Script/FMJamSessionRuntime.JamSessionComponent_MusicManager" (Variables|Default|GetJamSession)))
  (bind _participanttagaddedlistenerhandle (Variables|Default|GetParticipantTagAddedListenerHandle))
  ;[0 0.2882 1] Ready to play
  (Fortnite|FortShippingLog self "[GAB_Emote_JamLoop - OnParticipantTagAdded]")
  ;;[1 0.5029 0.5029] Is player able to join this jam part - Jam privacy checks
  (bind _returnvalue_3 (Class|JamSystemBPFL|IsJamPrivacyEnabled))
  (if _returnvalue_3
    (Utilities|IsValid _returnvalue
      (:Is%20Valid
        (bind _returnvalue_4 (Mojo|JamSession|GetJamSessionforActor _returnvalue))
        (if (Utilities|IsValidInterface _returnvalue_4)
          (if (not (Utilities|Enum|Equal(Enum) (Variables|/Script/FMJamSessionRuntime.JamSessionComponent_MusicManager|GetCurrentJamPrivacy (Class|JamSessionInterface|GetMusicManager _returnvalue_4)) "Public"))
            (bind _allowedtojoin (Default|IsPlayerAllowedtoJoinClosedJam))
            (if (not _allowedtojoin)
              (CallFunction|HandleJamError (Variables|Default|GetPlayerPawn) true "(GameplayTags=((TagName=\"FMJam.Playspace.NotAllowedToJoin.PrivateJam\")))" true)))
          (Utilities|FlowControl|Sequence
            (:then_0
              (if (Utilities|IsValidInterface (Variables|Default|GetJamSession))
                (Variables|Default|SetMusicManager _returnvalue_2)
                (Variables|Default|SetPlayerNetId (Mojo|JamSession|GetNetIdforPawn (Variables|Default|GetJamSession) (Variables|Default|GetPlayerPawn)))
                (CallFunction|BindPlayspaceEvents)
                (Utilities|FlowControl|Sequence
                  (:then_0
                    ;;[1 1 1]
                    (if (Class|JamCoreBPFL|ShouldRunServerCode (Utilities|Casting|CastToActor (Variables|Default|GetJamSession)))
                      (CallFunction|PlayLoopfromEID (select (Variables|Default|GetActivationFromQuickPlay) "QuickJoin" "Manual"))
                      (Variables|Default|SetActivationFromQuickPlay false)))
                  (:then_1
                    ;;[1 1 1] if they try to play a loop with licensed music turned off, tell them
                    (if (Class|JamCoreBPFL|ShouldRunClientCode (Utilities|Casting|CastToActor (Variables|Default|GetJamSession)))
                      (bind _returnvalue_5 (Fortnite|Audio|GetLicensedAudioTreatment))
                      (if (and (Utilities|Enum|NotEqual(Enum) _returnvalue_5 "None") (> (Fortnite|GetNumericCVar "FMJam.RespectLicensedMusicSetting") 0))
                        (CallFunction|HandleJamError (Variables|Default|GetPlayerPawn) true "(GameplayTags=((TagName=\"FMJam.Playspace.LicensedAudioDisabled\")))")))))
                (else
                  (bind _output_get (Variables|Default|SetLastCheckedNotifierComp (Actor|GetComponentbyClass (CallFunction|GetAvatarasFortPlayerPawn) "/Script/FMJamSessionRuntime.ActorComponent_JamSessionNotifier")))
                  (Utilities|IsValid _output_get
                    (:Is%20Valid
                      (bind _returnvalue_6 (Class|JamSessionObserverInterface|GetCurrentJamSession(Message) (Variables|Default|GetLastCheckedNotifierComp))
                        (:True
                          (CallFunction|FoundJamSession "EnteringSession" _returnvalue_6))
                        (:False
                          (EventDispatchers|BindEventtoOnJamSessionChangedBP (Variables|Default|GetLastCheckedNotifierComp) (delegate Custom|Found%20Jam%20Session)))))))))
            (:then_1
              ;;[1 1 1] Stop listening, we have already handled it
              (Utilities|IsValid _participanttagaddedlistenerhandle
                (:Is%20Valid
                  (AsyncAction|Cancel _participanttagaddedlistenerhandle)
                  (Variables|Default|SetParticipantTagAddedListenerHandle 0))))))))
    (else
      _)))

(event Custom|Found%20Jam%20Session
  (params (EventType :type EJamSessionNotificationType%20Enum :default "EnteringSession") (JamSession :type Jam%20Session%20Interface%20Interface))
  (bind (_outfoundcomponent _returnvalue) (Class|JamSessionComponentLibrary|GetComponentOnJamSession "/Script/FMJamSessionRuntime.JamSessionComponent_MusicManager" (Variables|Default|GetJamSession)))
  ;[0 0.2882 1] Ready to play
  (if (Utilities|Enum|Equal(Enum) EventType "EnteringSession")
    (bind _output_get (Variables|Default|SetJamSession JamSession))
    (Variables|Default|SetJamPlayspace (Utilities|Casting|CastToJamPlayspace _output_get))
    (if (Utilities|IsValidInterface (Variables|Default|GetJamSession))
      (CallFunction|UnbindNotifierEvent)
      (Variables|Default|SetMusicManager _returnvalue)
      (Variables|Default|SetPlayerNetId (Mojo|JamSession|GetNetIdforPawn (Variables|Default|GetJamSession) (Variables|Default|GetPlayerPawn)))
      (CallFunction|BindPlayspaceEvents)
      (Utilities|FlowControl|Sequence
        (:then_0
          ;;[1 1 1]
          (if (Class|JamCoreBPFL|ShouldRunServerCode (Utilities|Casting|CastToActor (Variables|Default|GetJamSession)))
            (CallFunction|PlayLoopfromEID (select (Variables|Default|GetActivationFromQuickPlay) "QuickJoin" "Manual"))
            (Variables|Default|SetActivationFromQuickPlay false)))
        (:then_1
          ;;[1 1 1] if they try to play a loop with licensed music turned off, tell them
          (if (Class|JamCoreBPFL|ShouldRunClientCode (Utilities|Casting|CastToActor (Variables|Default|GetJamSession)))
            (bind _returnvalue_2 (Fortnite|Audio|GetLicensedAudioTreatment))
            (if (and (Utilities|Enum|NotEqual(Enum) _returnvalue_2 "None") (> (Fortnite|GetNumericCVar "FMJam.RespectLicensedMusicSetting") 0))
              (CallFunction|HandleJamError (Variables|Default|GetPlayerPawn) true "(GameplayTags=((TagName=\"FMJam.Playspace.LicensedAudioDisabled\")))")))))
      (else
        (Fortnite|FortShippingLog self "Could not find playspace. Ending Ability")
        (Ability|EndAbility)))))

(event Custom|Unbind%20Notifier%20Event
  ;[0 0.2882 1] Ready to play
  (bind _lastcheckednotifiercomp (Variables|Default|GetLastCheckedNotifierComp)
    (:then
      (EventDispatchers|UnbindEventfromOnJamSessionChangedBP _lastcheckednotifiercomp (delegate Custom|Found%20Jam%20Session)))
    (:else)))

(event Custom|OnParticipantTagRemovedServer
  ;[1 1 1] Comment
  (bind _persistent (CallFunction|IsJamPlayspacePersistent))
  (if (not _persistent)
    (Fortnite|FortShippingLog self "Member tag removed. Ending Ability")
    (Ability|EndAbility)))

(event Custom|OnBlockJamEmotesTagAddedServer
  ;[1 1 1] If we received the block jam emotes tag, then stop playing
  (Fortnite|FortShippingLog self "BlockJam Tag added. Ending Ability")
  (Ability|EndAbility))

(event Custom|HandleOnAutoRunEnabled
  (params (bAutoRunEnabled :type Boolean))
  ;[1 1 1] End Jam traversal if we auto-run out of it on mobile
  (if bAutoRunEnabled
    (if (Inputs|IsUsingInputType "Touch" (Fortnite|GetFortPlayerControllerfromActor (Variables|Default|GetPlayerPawn)))
      (Ability|EndAbility))))

(fn GetCacheContextOverride
  (return 0))

(fn GetLoopType
  ;[1 1 1] As we migrate away from EID, we might instead be provided the song loop type via activation parameters... so, in that case, don't try to fetch it from the EID (CurrentSourceObject)
  (bind (_outpickablesongtargetdata _returnvalue) (Class|SparksPickableSongPartHelpers|TryGetPickableSongTargetDatafromGameplayEventTargetDataHandle (Utilities|Struct|BreakGameplayEventData (Variables|Ability|GetCurrentEventData))))
  (if _returnvalue
    (return (Utilities|Struct|BreakSparksPickableSongPartTargetData _outpickablesongtargetdata))
    (else
      ;[1 1 1] Otherwise, we can try to get it from the Current Source Object
      (bind _assparks_jam_emote_item_definition (Utilities|Casting|CastToSparksJamEmoteItemDefinition (Ability|GetCurrentSourceObject))
        (:then
          (return (Class|SparksJamEmoteItemDefinition|GetJamLoopType _assparks_jam_emote_item_definition)))
        (:CastFailed
          ;[1 0 0] Could not get EID
          (Development|PrintText "INVTEXT(\"Could not find EID\")" true true "(R=1.000000,G=0.000000,B=0.103144,A=1.000000)")
          (return "Num"))))))

(fn GetSongShortName
  ;[1 1 1] As we migrate away from EID, we might instead be provided the song short name via activation parameters... so, in that case, don't try to fetch it from the EID (CurrentSourceObject)
  (bind (_outpickablesongtargetdata _returnvalue) (Class|SparksPickableSongPartHelpers|TryGetPickableSongTargetDatafromGameplayEventTargetDataHandle (Utilities|Struct|BreakGameplayEventData (Variables|Ability|GetCurrentEventData))))
  (if _returnvalue
    (return (Utilities|Struct|BreakSparksPickableSongPartTargetData _outpickablesongtargetdata))
    (else
      ;[1 1 1] Otherwise, we can try to get it from the Current Source Object
      (bind _assparks_jam_emote_item_definition (Utilities|Casting|CastToSparksJamEmoteItemDefinition (Ability|GetCurrentSourceObject))
        (:then
          (return (Class|SparksJamEmoteItemDefinition|GetJamSongShortName _assparks_jam_emote_item_definition)))
        (:CastFailed
          ;[1 0 0] Could not get EID
          (Development|PrintText "INVTEXT(\"Could not find EID\")" true true "(R=1.000000,G=0.000000,B=0.103144,A=1.000000)")
          (return "None"))))))

(fn IsJamPlayspacePersistent
  (bind _jamplayspace (Variables|Default|GetJamPlayspace)
    (:then
      (bind _asbp_jam_playspace_volume (Utilities|Casting|CastToBP_JamPlayspace_Volume (Variables|/Script/PlayspaceSystem.Playspace|GetBoundGameplayVolume _jamplayspace)))
      (return (Class|BPJamPlayspaceVolume|IsPersistent _asbp_jam_playspace_volume)))
    (:else
      (bind _asjam_session_actor (Utilities|Casting|CastToJamSessionActor (Variables|Default|GetJamSession))
        (:then
          (return (Variables|/Script/FMJamSessionRuntime.JamSessionActor|GetbIsPersistent _asjam_session_actor)))
        (:CastFailed
          (return false))))))

(fn IsPlayerInInnerJamRing
  (params (Player%20Pawn :type Fort%20Player%20Pawn%20Object%20Reference))
  (bind _jamsession (Variables|Default|GetJamSession))
  (if (Utilities|IsValidInterface _jamsession)
    (return (Class|JamSessionInterface|IsPlayerinInnerCollider _jamsession (Variables|GetPlayer%20Pawn)))
    (else
      (return false))))

(fn IsJamPlayspaceBlocked
  (bind _jamplayspace (Variables|Default|GetJamPlayspace)
    (:then
      (bind _boundgameplayvolume (Variables|/Script/PlayspaceSystem.Playspace|GetBoundGameplayVolume _jamplayspace)
        (:then
          (bind _asbp_jam_playspace_volume (Utilities|Casting|CastToBP_JamPlayspace_Volume _boundgameplayvolume)
            (:then
              (bind _blockjam (Class|BPJamPlayspaceVolume|ShouldBlockJam _asbp_jam_playspace_volume))
              (if _blockjam
                (bind _errormessage (Class|BPJamPlayspaceVolume|GetBlockedJamMessage _asbp_jam_playspace_volume))
                (return true _errormessage)
                (else
                  _)))
            (:CastFailed
              _))
          (return false 0))
        (:else
          _)))
    (:else
      _)))

(fn IsPlayerAllowedToJoinClosedJam
  (Utilities|FlowControl|Sequence
    (:then_0
      (bind _returnvalue (Mojo|JamSession|GetJamSessionforActor (Ability|GetActivatingPawn)))
      (bind _asfort_player_state_athena (Utilities|Casting|CastToFortPlayerStateAthena (Variables|/Script/Engine.Pawn|GetPlayerState (Ability|GetActivatingPawn)))
        (:then
          (return (Mojo|JamSession|IsPlayerinSamePartyasAnyJammer (Class|JamSessionInterface|GetMusicManager _returnvalue) _asfort_player_state_athena)))
        (:CastFailed
          (Fortnite|FortShippingLog self "PlayerState for Activating Pawn failed to cast to FortPlayerStateAthena" true))))
    (:then_1
      (return (LocalVariables|GetCanJoin)))))

(fn JoinOrCreatePlayspace
  (bind _playerpawn (Variables|Default|GetPlayerPawn))
  (bind _capsulecomponent (Variables|/Script/Engine.Character|GetCapsuleComponent _playerpawn))
  (bind _returnvalue (Components|Capsule|GetScaledCapsuleHalfHeight _capsulecomponent))
  (bind _returnvalue_2 (Math|Vector|MakeVector 0.0 0.0 _returnvalue))
  (bind _returnvalue_3 (Transformation|GetActorTransform _playerpawn))
  (bind _rotation (.rotation _returnvalue_3))
  (bind _returnvalue_4 (Mojo|JamSession|GetJamSessionforActor (Variables|Default|GetPlayerPawn)))
  (if (Utilities|IsValidInterface _returnvalue_4)
    ;[0.2874 0.1221 0.2623 0.502] Special-Casing JamSessionActors...
    ;
    ; Part 1: Players that aren't within the InnerCollider can spawn new JamSessionActors...
    ;
    ; See "SpawnJamPlayspaceAt" for part 2
    (Class|JamOverrides|ShouldSpawnJamasSession
      (:True
        (bind _output_get (Variables|Default|SetJamSession _returnvalue_4)) ; Update JamSession here so Is Player In Inner Jam Ring will work\nWe may override JamSession in a bit, that's ok
        (bind _inside (CallFunction|IsPlayerinInnerJamRing (Variables|Default|GetPlayerPawn)))
        (if _inside
          _
          (elif (Utilities|Array|ContainsItem (Mojo|JamSession|GetAllJammers (Class|JamSessionInterface|GetMusicManager _output_get)) (Mojo|JamSession|GetNetIdforPawn _output_get (Variables|Default|GetPlayerPawn)))
            _
            (else
              ;[0 0.7292 0.3064] Spawn a new JamSession: Look for the ground under the player. Start at the player's actor location (center of capsule) and search a specified distance below their feet
              ;;[1 0.9309 0.1394 0.5] Add an offset so that we can control where the jam circle spawns in relation to the player.
              (bind _output_get_2 (LocalVariables|SetFinalPlayspaceOffset (+ (.location _returnvalue_3) (Math|Vector|RotateVector (Variables|Default|GetBasePlayspaceOffset) _rotation))))
              (bind (_outhit _returnvalue_5) (Collision|LineTraceByChannel _output_get_2 (- (- _output_get_2 _returnvalue_2) (Math|Vector|MakeVector 0.0 0.0 (Variables|Default|GetGroundSearchHeight))) "TraceTypeQuery10" true (Utilities|Array|MakeArray (Variables|Default|GetPlayerPawn)) "ForDuration" true "(R=1.000000,G=1.000000,B=1.000000,A=1.000000)" "(R=0.000000,G=1.000000,B=0.000000,A=1.000000)" 2e+01))
              (if _returnvalue_5
                ;[0 0.7292 0.3064] Create the playspace at the found collision location - with the player rotation
                (CallFunction|SpawnJamPlayspaceAt (Math|Transform|MakeTransform (.Location (Collision|BreakHitResult _outhit)) _rotation) (Variables|Default|GetPlayerPawn))
                ;[1 0.5 0.5] JamPlayspace created
                (return)
                (else
                  ;[0 0.7292 0.3064] When no ground is found - use the players location and rotation (NOT scale) at their feet
                  (Development|PrintString "No Collision for Jam playspace - use player location" true true "(R=1.000000,G=1.000000,B=0.000000,A=1.000000)" 1e+01 "PlayspaceRayTest")
                  (CallFunction|SpawnJamPlayspaceAt (Math|Transform|MakeTransform (- _output_get_2 _returnvalue_2) _rotation) (Variables|Default|GetPlayerPawn))
                  ;[1 0.5 0.5] JamPlayspace created
                  (return)))))))
      (:False
        _))
    ;[1 0.5 0.5] Player was already in a JamSession, and we don't want to spawn another one
    (return)
    (else
      _)))

(fn CanActivateAbility
  (params (ActorInfo :type Gameplay%20Ability%20Actor%20Info%20Structure) (Handle :type Gameplay%20Ability%20Spec%20Handle%20Structure))
  (bind (_owneractor _avataractor _playercontroller) (Utilities|Struct|BreakGameplayAbilityActorInfo ActorInfo))
  (bind (_relevanttags _returnvalue) (|CallParentFunction ActorInfo Handle))
  (if _returnvalue
    (bind _asfort_player_controller (Utilities|Casting|CastToFortPlayerController _playercontroller)
      (:then
        (Class|JamSystemBPFL|IsJammingBlocked _asfort_player_controller
          (:True
            _)
          (:False
            (bind _returnvalue_2 (Class|JamSystemBPFL|IsLoopTypeUsablebyEmote (CallFunction|GetLoopType) _asfort_player_controller))
            (if _returnvalue_2
              (bind _returnvalue_3 (Mojo|JamSession|IsAllowedtoJaminSplitscreen _asfort_player_controller (Default|GetGameAccountId _asfort_player_controller)))
              (if _returnvalue_3
                (return _relevanttags true)
                (else
                  (GameplayTags|AddGameplayTag _relevanttags "(TagName=\"FMJam.Playspace.NotPrimaryPlayer\")" :!class /Script/GameplayTags.BlueprintGameplayTagLibrary)))
              (else
                _))
            (return _relevanttags false))))
      (:CastFailed
        _))
    (else
      _)))

(fn OnMontageStarted
  (params (Montage :type Anim%20Montage%20Object%20Reference))
  (if (> (Fortnite|GetNumericCVar "FMJam.EmotesCancelJam") 0)
    ;[1 1 1] Simulating normal Emotes that are ended when another montage is played. We don't use montages for Jam so this has to be done manually.
    (Ability|EndAbility)
    ;[1 0.5 0.5] Jam cancelled while emoting
    (return)
    (else
      ;[1 0.5 0.5] Jam not cancelled when emoting
      (return))))

(fn PlayLoopFromEID
  (params (JamStartMethod :type EJamSlotStartMethod%20Enum :default "Manual"))
  ;[1 0.5 0.5] We can only request songs to be played on the server
  (Ability|HasAuthority
    (:True
      ;[1 1 1] fail if we're in a blocked jam area
      (bind (_isblocked _errormessage) (CallFunction|IsJamPlayspaceBlocked))
      (if _isblocked
        (CallFunction|HandleJamError (Variables|Default|GetPlayerPawn) true _errormessage true)
        (return false)
        (else
          ;[1 1 1] fail if they aren't in the middle jam ring
          (bind _inside (CallFunction|IsPlayerinInnerJamRing (Variables|Default|GetPlayerPawn)))
          (if _inside
            ;[1 0.5029 0.5029] Is the player able to join this playspace (or were they already in it)
            (bind _musicmanager (Variables|Default|GetMusicManager)
              (:then
                (if (or (not (Mojo|JamSession|IsJamFullwithParams _musicmanager (Mojo|JamSession|GetSearchParamsforIsJamFull))) (Utilities|IsValid (Mojo|JamSession|GetMusicSlotforPlayerId _musicmanager (Variables|Default|GetPlayerNetId)) :!class /Script/Engine.KismetSystemLibrary))
                  ;[1 0.5 0.5] Try to play the loop
                  (bind _returnvalue (Mojo|JamSession|PlayLoop (Variables|Default|GetMusicManager) (Variables|Default|GetPlayerNetId) (Default|GetSongShortName) (CallFunction|GetLoopType) JamStartMethod false false))
                  (Utilities|IsValid _returnvalue
                    (:Is%20Valid
                      ;[1 0.5 0.5] The Loop has been played
                      (return true))
                    (:Is%20Not%20Valid
                      ;[1 0.5 0.5] The loop failed to play
                      (CallFunction|HandleJamError (Variables|Default|GetPlayerPawn) true "(GameplayTags=((TagName=\"FMJam.Playspace.JamFull\")))" true)
                      (return false)))
                  (else
                    (CallFunction|HandleJamError (Variables|Default|GetPlayerPawn) true "(GameplayTags=((TagName=\"FMJam.Playspace.JamFull\")))" true)
                    (return false))))
              (:else))
            (else
              (CallFunction|HandleJamError (Variables|Default|GetPlayerPawn) true "(GameplayTags=((TagName=\"FMJam.Playspace.NotInJamRing\")))" true)
              (return false))))))
    (:False
      (return false))))

(fn ResetPropertiesOnActivation
  ;[1 1 1] As long as we are InstancedPerActor rather than InstancedPerExecution, we need to make sure any runtime properties that aren't otherwise cleaned up in OnEndAbility get reset to their defaults here
  (Variables|Default|SetJamPlayspace 0)
  (Variables|Default|SetJamSession 0)
  (Variables|Default|SetMusicManager 0)
  (Variables|Default|SetAbilityEnded false)
  (Variables|Default|SetLoopInstanceId -1))

(fn ShouldPlayFailedMontage
  (params (FailedReason :type Gameplay%20Tag%20Container%20Structure%20%28by%20ref%29))
  (|CallParentFunction FailedReason)
  (return true))

(fn SpawnJamPlayspaceAt
  (params (Transform :type Transform) (CreatedBy :type Fort%20Player%20Pawn%20Object%20Reference))
  (bind _jamplayspacevolume (LocalVariables|GetJamPlayspaceVolume))
  (Utilities|FlowControl|Sequence
    (:then_0
      (bind _returnvalue (Class|PlayspaceLibrary|GetPlayspaceforActor CreatedBy CreatedBy))
      (Utilities|IsValid _returnvalue
        (:Is%20Valid
          _)
        (:Is%20Not%20Valid
          (Fortnite|FortShippingLog self "No Playspace/Root exists yet" true)))
      (LocalVariables|SetOverridesComp (Actor|GetComponentbyClass (Game|GetGameState) "/Script/FMJamCoreRuntime.JamOverrides"))
      (Class|JamOverrides|ShouldSpawnJamasSession
        (:True
          (bind _overridescomp (LocalVariables|GetOverridesComp)
            (:then
              (bind _jamsessionclass (Variables|/Script/FMJamCoreRuntime.JamOverrides|GetJamSessionClass _overridescomp)
                (:then
                  (bind _returnvalue_2 (Game|SpawnActorfromClass _jamsessionclass (Variables|GetTransform) "AlwaysSpawn" "MultiplyWithRoot" 0 (Variables|GetCreatedBy))) ; Todo: Should we actually immediately capture this as a var?
                  (Utilities|Casting|CastToJamSessionInterface _returnvalue_2
                    (:then
                      (bind _asjam_session_actor (Utilities|Casting|CastToJamSessionActor _returnvalue_2))
                      (LocalVariables|SetSpawnedJamSession _asjam_session_actor))
                    (:CastFailed
                      (Actor|DestroyActor _returnvalue_2) ; Destroy the actor if it wasn't a JamSession
                    )))
                (:else
                  (bind _returnvalue_3 (Game|SpawnActorfromClass "/FMJamSystem/Session/BP_JamSessionActor_Sphere.BP_JamSessionActor_Sphere_C" (Variables|GetTransform) "AlwaysSpawn" "MultiplyWithRoot" 0 (Variables|GetCreatedBy))) ; Todo: Should we actually immediately capture this as a var?
                  (LocalVariables|SetSpawnedJamSession _returnvalue_3)))
              ;[0.2874 0.1221 0.2623 0.502] Special-Casing JamSessionActors...
              ;
              ; Part 2: JamSessionActors that are too close to other JamSessions shouldn't spawn
              ;
              ; See "JoinOrCreatePlayspace" for part 1
              (bind _spawnedjamsession (LocalVariables|GetSpawnedJamSession)
                (:then
                  (bind _returnvalue_4 (Mojo|JamSession|IsPlacementValid _spawnedjamsession))
                  (if _returnvalue_4
                    (Variables|Default|SetJamSession _spawnedjamsession) ; Immediately, track this new JamSession as a JamSession
                    (else
                      (Actor|DestroyActor _spawnedjamsession)
                      (CallFunction|HandleJamError (Variables|Default|GetPlayerPawn) true "(GameplayTags=((TagName=\"FMJam.Playspace.TooCloseToOtherJamRing\")))" true)
                      (Development|PrintString "Failed to create a JamSession" true true "(R=1.000000,G=0.000000,B=0.000000,A=1.000000)")
                      (return))))
                (:else)))
            (:else
              _)))
        (:False
          ;[0.5284 1 0.6603] Spawn the playspace volume (which will spawn the playspace)
          (bind _overridescomp_2 (LocalVariables|GetOverridesComp)
            (:then
              (bind _jamvolume (Variables|/Script/FMJamCoreRuntime.JamOverrides|GetJamVolume _overridescomp_2)
                (:then
                  (bind _returnvalue_5 (Game|SpawnActorfromClass _jamvolume (Variables|GetTransform) "AlwaysSpawn" "OverrideRootScale" 0 (Variables|GetCreatedBy))) ; Create a new collab point
                  ;[1 1 1] The override case needs special handling, due to the input class type being Gameplay Volume and not BP_JamPlayspaceVolume / etc
                  (bind _asbp_jam_playspace_volume (Utilities|Casting|CastToBP_JamPlayspace_Volume _returnvalue_5)
                    (:then
                      (LocalVariables|SetJamPlayspaceVolume _asbp_jam_playspace_volume))
                    (:CastFailed
                      ;;[1 1 1] If the Playspace Volume isn't of the correct type, then remove it.
                      (Utilities|IsValid (Class|GameplayVolume|GetPlayspace _returnvalue_5)
                        (:Is%20Valid
                          (Actor|DestroyActor (Class|GameplayVolume|GetPlayspace _returnvalue_5)))
                        (:Is%20Not%20Valid
                          _))
                      (Actor|DestroyActor _returnvalue_5)
                      (return))))
                (:else
                  _)))
            (:else
              _))
          (if (> (Fortnite|GetNumericCVar "FMJam.PlayspaceShape.Sphere") 0)
            (bind _returnvalue_6 (Game|SpawnActorfromClass "/FMJamSystem/Playspace/BP_JamPlayspace_Volume_Sphere.BP_JamPlayspace_Volume_Sphere_C" (Variables|GetTransform) "AlwaysSpawn" "OverrideRootScale" 0 (Variables|GetCreatedBy))) ; Create a new collab point
            (LocalVariables|SetJamPlayspaceVolume _returnvalue_6)
            (else
              (bind _returnvalue_7 (Game|SpawnActorfromClass "/FMJamSystem/Playspace/BP_JamPlayspace_Volume_Round.BP_JamPlayspace_Volume_Round_C" (Variables|GetTransform) "AlwaysSpawn" "MultiplyWithRoot" 0 (Variables|GetCreatedBy)))
              (LocalVariables|SetJamPlayspaceVolume _returnvalue_7))))))
    (:then_1
      ;[0.5284 1 0.6603] Verify that this playspace was created in a valid loation
      (Utilities|IsValid _jamplayspacevolume
        (:Is%20Valid
          (bind (_out_blockederrormessage _returnvalue_8) (JamSystem|IsJamVolumePlacementValid _jamplayspacevolume))
          (if _returnvalue_8
            (bind _asjam_playspace (Utilities|Casting|CastToJamPlayspace (Class|GameplayVolume|GetPlayspace _jamplayspacevolume))
              (:then
                (bind _output_get (Variables|Default|SetJamPlayspace _asjam_playspace))
                (bind _output_get_2 (Variables|Default|SetJamSession _output_get))
                (Variables|Default|SetMusicManager (.ReturnValue (Class|JamSessionComponentLibrary|GetComponentOnJamSession "/Script/FMJamSessionRuntime.JamSessionComponent_MusicManager" _output_get_2)))
                ;;[1 0.349 0.4266] Done
                (return))
              (:CastFailed
                ;;[1 0 0] Error handling
                (Actor|DestroyActor (LocalVariables|GetJamPlayspaceVolume))
                (CallFunction|HandleJamError (Variables|Default|GetPlayerPawn) true "(GameplayTags=((TagName=\"FMJam.Playspace.TooCloseToOtherJamRing\")))" true)))
            (else
              (Actor|DestroyActor _jamplayspacevolume)
              (CallFunction|HandleJamError (Variables|Default|GetPlayerPawn) true (GameplayTags|MakeGameplayTagContainerfromTag _out_blockederrormessage) true)
              ;;[1 0.5 0.5] Too close to other Jam
              (return))))
        (:Is%20Not%20Valid
          _))
      (Development|PrintString "Failed to create a JamPlayspace" true true "(R=1.000000,G=0.000000,B=0.000000,A=1.000000)")
      (return))))

(fn TryDisableMovement
  (bind _playerpawn (Variables|Default|GetPlayerPawn))
  ;[1 0.286 1] Only change the movement if the cvar allows it
  (if (== (Fortnite|GetNumericCVar "FMJam.EnableJamEmoteMovement") 0)
    ;[1 0.469 0.161] Change the movement mode and cache old movement
    (Utilities|IsValid _playerpawn
      (:Is%20Valid
        (Creative|GameplayEffects|SimpleApplyGameplayEffectToActor "/FMJamSystem/Player/Abilities/GE_FMJamBlockMovement.GE_FMJamBlockMovement_C" _playerpawn)
        (bind _asfort_player_controller (Utilities|Casting|CastToFortPlayerController (Pawn|GetController _playerpawn))
          (:then
            (Variables|Default|SetCachedRotatePawnToCamera (Input|GetRotatePawntoCamera _asfort_player_controller))
            (Input|SetRotatePawntoCamera _asfort_player_controller false)
            ;[1 0.5 0.5] Movement has been changed
            (return))
          (:CastFailed
            _)))
      (:Is%20Not%20Valid
        _))
    ;[1 0.005 0.005] We are missing something
    (return)
    (else
      ;[1 0.5 0.5] Movement disabling has been disabled
      (return))))

(fn TryEnableMovment
  (bind _returnvalue (Ability|GetAbilitySystemComponent (Variables|Default|GetPlayerPawn)))
  (bind _playerpawn (Variables|Default|GetPlayerPawn))
  ;[0.349 0.526 1] Restore movement
  (Utilities|IsValid _playerpawn
    (:Is%20Valid
      (Utilities|IsValid _returnvalue
        (:Is%20Valid
          (GameplayEffects|RemoveActiveGameplayEffectbySourceEffect _returnvalue "/FMJamSystem/Player/Abilities/GE_FMJamBlockMovement.GE_FMJamBlockMovement_C" 0 -1)
          (bind _asfort_player_controller (Utilities|Casting|CastToFortPlayerController (Pawn|GetController _playerpawn))
            (:then
              (if (Variables|Default|GetCachedRotatePawnToCamera)
                (Input|SetRotatePawntoCamera _asfort_player_controller true)
                (else
                  _))
              ;[1 0.5 0.5] Movement has been restored
              (return))
            (:CastFailed
              _)))
        (:Is%20Not%20Valid
          _))
      ;[1 0 0] Something is missing
      (return))
    (:Is%20Not%20Valid
      _)))
