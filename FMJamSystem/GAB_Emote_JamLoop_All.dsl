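; Jam loop emote ability that plays a whole song in a Jam Session: the player takes one part and other free slots get the rest.
; GetLoopType keeps the player's current part or picks an unused one by preference, and activation requires play-all usability.
; PlayLoopFromEID starts the loop as PlayAll, can push catalog tempo, key and mode to session audio for a solo player, then requests the slot fill.
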(event Ability|EventActivateAbility
  (|CallParentFunction) ; This node is disabled and will not be called.\nDrag off pins to build functionality.
)

(event Ability|EventOnEndAbility
  (params (bWasCancelled :type Boolean))
  (|CallParentFunction bWasCancelled) ; This node is disabled and will not be called.\nDrag off pins to build functionality.
)

(event EventOnOwnJamLoopStartedEvent
  (params (LoopInstanceId :type Integer))
  (|CallParentFunction LoopInstanceId))

(fn GetAudioStateForCurrentSong
  (bind _returnvalue (GameInstanceSubsystems|GetSparksSongCatalog))
  (Utilities|IsValid _returnvalue
    (:Is%20Valid
      (bind _returnvalue_2 (Class|SparksSongCatalog|GetSongbyShortName _returnvalue (Utilities|String|ToString(Name) (Default|GetSongShortName))))
      (Utilities|IsValid _returnvalue_2
        (:Is%20Valid
          (bind (_dataentry_songuid _dataentry_shortname _dataentry_title _dataentry_artist _dataentry_album _dataentry_genre _dataentry_animgenre _dataentry_stagemood _dataentry_duration _dataentry_countinduration _dataentry_musicmomentoffsetseconds _dataentry_releaseyear _dataentry_newuntiltime _dataentry_agerating _dataentry_musickey _dataentry_musicmode _dataentry_musictempo) (Utilities|Struct|BreakSparksSongCatalogEntry (Variables|/Script/SparksCMS.CatalogData|GetDataEntry_SongUID _returnvalue_2)))
          (return _dataentry_musictempo _dataentry_musickey _dataentry_musicmode true))
        (:Is%20Not%20Valid
          _)))
    (:Is%20Not%20Valid
      _))
  (return 0 "C" "Major" false))

(fn GetLoopType
  (bind _playernetid (Variables|Default|GetPlayerNetId))
  (bind _music_manager (LocalVariables|Getmusic%20manager))
  (bind _returnvalue (Mojo|JamSession|GetMusicSlotforPlayerId _music_manager _playernetid))
  (if (Utilities|IsValidInterface (Variables|Default|GetJamSession))
    (bind _output_get (LocalVariables|Setmusic%20manager (Class|JamSessionInterface|GetMusicManager (Variables|Default|GetJamSession))))
    (Utilities|IsValid _output_get
      (:Is%20Valid
        ; Try to use the same loop type that the player is currently using
        (Utilities|IsValid _returnvalue
          (:Is%20Valid
            ;[1 0.5 0.5 0.5] Found the player's loop type
            (return (Class|JamMusicSlot|GetCurrentLoopType _returnvalue)))
          (:Is%20Not%20Valid
            ; Set the "current LoopType" to be the first one that isn't currently being used in the JamSession. This is the one that the player will play before the holojammers get their turn.
            (bind _output_get_2 (LocalVariables|SetUnusedLoopTypes (Mojo|JamSession|GetUnusedLoopTypes (LocalVariables|Getmusic%20manager))))
            (if (Utilities|Array|IsNotEmpty _output_get_2)
              (for _array_element (LocalVariables|GetPlayerLoopTypePreference)
                (if (Utilities|Array|ContainsItem (LocalVariables|GetUnusedLoopTypes) _array_element)
                  ;[1 0.5 0.5 0.5] Found an unused loop
                  (return _array_element)))
              ;[1 0 0 0.5] We shouldn't have been able to get here! Maybe the PlayerLoopTypePreference is incomplete?
              (return (Utilities|Array|Get(acopy) (LocalVariables|GetUnusedLoopTypes)))
              (else
                ;[1 0 0 0.5] Could not find a loop type, call the parent func and hope for the best
                (return (|CallParentFunction)))))))
      (:Is%20Not%20Valid
        _))
    (else
      _))
  ;[1 0.5 0.5 0.5] There is no JamSession, use the first preference
  (return (Utilities|Array|Get(acopy) (LocalVariables|GetPlayerLoopTypePreference))))

(fn CanActivateAbility
  (params (ActorInfo :type Gameplay%20Ability%20Actor%20Info%20Structure) (Handle :type Gameplay%20Ability%20Spec%20Handle%20Structure))
  (bind (_owneractor _avataractor _playercontroller) (Utilities|Struct|BreakGameplayAbilityActorInfo ActorInfo))
  (bind (_relevanttags _returnvalue) (|CallParentFunction ActorInfo Handle))
  (if _returnvalue
    (bind _asfort_player_controller (Utilities|Casting|CastToFortPlayerController _playercontroller)
      (:then
        ; Specifically check None (PlayAll) here since the parent call will end up checking the missing part only, not necessarily PlayAll
        (bind _returnvalue_2 (Class|JamSystemBPFL|IsLoopTypeUsablebyEmote "Num" _asfort_player_controller))
        (return _relevanttags _returnvalue_2))
      (:CastFailed
        _))
    (else
      _))
  (return _relevanttags false))

(fn PlayLoopFromEID
  (params (JamStartMethod :type EJamSlotStartMethod%20Enum :default "Invalid"))
  (bind _jamsession (Variables|Default|GetJamSession))
  (bind (_outfoundcomponent _returnvalue) (Class|JamSessionComponentLibrary|GetComponentOnJamSession "/Script/FMJamSessionRuntime.JamSessionComponent_AudioState" _jamsession))
  (bind _jamsession_2 (Variables|Default|GetJamSession))
  (bind _returnvalue_2 (Class|JamSessionInterface|GetMusicManager _jamsession_2))
  ; Play the first loop on the player but change the start method to Play All
  (bind _requestedsong (|CallParentFunction "PlayAll"))
  (if _requestedsong
    (if (Utilities|IsValidInterface (Variables|Default|GetJamSession))
      (Utilities|FlowControl|Sequence
        (:then_0
          ; If we are the only player, update the Audio State
          (switch int (Mojo|JamSession|CountActiveSlotswithParams _returnvalue_2 (struct :ContextPlayer "INVALID" :SpecificLoopType "Num" :AutojammerMode "ExcludeAll"))
            (:0
              _)
            (:1
              (Utilities|IsValid (Mojo|JamSession|GetMusicSlotforPlayerId _returnvalue_2 (Variables|Default|GetPlayerNetId))
                (:Is%20Valid
                  _))))
          (LocalVariables|SetIsSinglePlayerChange true))
        (:then_1
          (if (LocalVariables|GetIsSinglePlayerChange)
            (bind (_tempo _key _mode _founddata) (CallFunction|GetAudioStateforCurrentSong))
            (if _founddata
              (if _outfoundcomponent
                (Class|JamSessionComponentAudioState|TriggerTempoChange _returnvalue _tempo)
                (Class|JamSessionComponentAudioState|TriggerKeyChange _returnvalue _key)
                (Class|JamSessionComponentAudioState|TriggerModeChange _returnvalue _mode)))))
        (:then_2
          (Mojo|JamSession|RequestExternalFillJamSlotswithSong (Class|JamSessionInterface|GetMusicManager (Variables|Default|GetJamSession)) (Variables|Default|GetPlayerNetId) (Default|GetSongShortName))
          ;[1 0.5 0.5 0.5] All free slots should be playing the song now
          (return true)))
      (else
        ;[1 0 0 0.5] Could not find the JamSession
        (return false)))
    (else
      ;[1 0 0 0.5] The player could not play a loop, so we won't attempt to fill other slots
      (return false))))
