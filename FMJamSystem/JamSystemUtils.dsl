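; Blueprint function library of helpers for the FM Jam System's loop types.
; GetLoopTextFromLoopType wraps the native Mojo call that returns display text for an EFMJamLoopType value, defaulting to Lead.
; GetLoopTypeFromTag resolves a gameplay tag to a loop type and also returns whether that result differs from the Num sentinel, signalling a valid match.
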(fn GetLoopTextFromLoopType
  (params (LoopType :type EFMJamLoopType%20Enum :default "Lead"))
  (bind _returnvalue (Mojo|JamSystem|GetLoopTypeText LoopType))
  (return _returnvalue))

(fn GetLoopTypeFromTag
  (params (GampelayTag :type Gameplay%20Tag%20Structure))
  (bind _returnvalue (Mojo|JamSystem|GetLoopTypebyTag GampelayTag))
  (return _returnvalue (Utilities|Enum|NotEqual(Enum) _returnvalue "Num")))
