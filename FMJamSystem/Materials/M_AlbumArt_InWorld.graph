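# M_AlbumArt_InWorld is a Default Lit material in FMJamSystem, named for showing album art on in-world surfaces.
# It samples the AlbumArt texture parameter and multiplies its RGB by the DisplayAlbumArt scalar parameter, which can scale or black out the image.
# The graph lists no other nodes, and automatic usage-flag setting in the editor is turned off.
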Material /FMJamSystem/Materials/M_AlbumArt_InWorld
graph {
 properties {bAutomaticallySetUsageInEditor=false, usedShadingModels=Default Lit}
 nodes [Multiply(A=TextureSample(Tex=TextureObjectParameter() {parameterName=AlbumArt}).RGB, B=ScalarParameter() {parameterName=DisplayAlbumArt})]
}
