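# Base translucent, unlit deferred decal material that draws a ring for the Jam playspace.
# An outer FX_SphereMask2D times an inverted inner one forms the ring, with radii and hardness scaled by Dynamic Parameters.
# A faint copy of the ring plus an eroded inner-edge mask is modulated by tiled tech lines pulsing on a sine of time.
# The ring mask also drives a parameterized emissive color with strength and boost multipliers.
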Material /FMJamSystem/Materials/M_JamPlayspace_Decal_Base
graph {
 properties {
  blendMode=BLEND_Translucent
  materialDomain=MD_DeferredDecal
  pixelDepthOffsetMode=PDOM_Legacy
  shadingModel=MSM_Unlit
  usedShadingModels=Unlit
 }
 nodes [
  n1 = DynamicParameter() {defaultValue=<rgba 0.2 0.9 0.3 0.9>, paramNames [InnerRing, InnerHardness, OuterRing, OuterHardness]}
  n2 = MaterialFunctionCall!(
   Hardness=Multiply(A=ScalarParameter() {defaultValue=1.05, parameterName=Radius_OuterHardness}, B=@n1.OuterHardness)
   Center
   Radius=Multiply(A=ScalarParameter() {defaultValue=1.55, parameterName=Radius_OutterRing}, B=@n1.OuterRing)
  ) {materialFunction=/Game/Packages/Fortress_SharedMaterials/Base_Material_Functions/FX_SphereMask2D}
  n3 = Multiply(A=ScalarParameter() {defaultValue=1.01, parameterName=Radius_InnerHardness}, B=@n1.InnerHardness)
  n4 = Multiply(A=ScalarParameter() {defaultValue=2.2, parameterName=Radius_InnerRing}, B=@n1.InnerRing)
  n5 = Multiply(
   A=@n2.Result
   B=OneMinus(None=MaterialFunctionCall(Hardness=@n3, Radius=@n4) {materialFunction=&FX_SphereMask2D}.Result)
  )
  n6 = MaterialFunctionCall(Hardness=Constant() {r=0.09}, Radius=ScalarParameter() {defaultValue=3.5, parameterName=InnerErodeStrength}) {materialFunction=&FX_SphereMask2D}
  n7 = Add(A=Multiply(A=@n5) {constB=0.1}, B=Multiply(A=@n2.Result, B=OneMinus(None=@n6.Result)))
  n8 = Multiply(
   A=TextureSample(UVs=TextureCoordinate() {uTiling=25, vTiling=25}) {texture=/Game/Effects/Fort_Effects/Textures/T_TechLines}.B
   B=ScalarParameter() {defaultValue=2, parameterName=PatternOpacity}
  )
  n9 = Time()
  n10 = Divide(A=Add(A=Sine(None=@n9) {period=2}))
  Add(A=@n7, B=Multiply(A=@n7, B=Multiply(A=@n8, B=@n10)))
  n11 = VectorParameter() {parameterName=EmissiveColor}
  Multiply(
   A=Multiply(A=Multiply(A=@n5, B=@n11.RGB), B=ScalarParameter() {parameterName=EmissiveStrength})
   B=ScalarParameter() {defaultValue=1, parameterName=EmissiveBoost}
  )
  VectorParameter() {parameterName=BaseColor}
 ]
}
