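# UI material for the Jam system that draws an instrument type icon picked by the Icon parameter.
# A chain of saturated lerps selects among SDF channels from two textures and a composite that arranges scaled Drums, Bass, Guitar and Vocals glyphs around the center.
# The selected shape is antialiased, with an optional gradient on its alpha.
# The Lines static switch swaps flat Icon Color for animated two-color ping-pong stripes and alters the alpha.
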Material /FMJamSystem/Materials/M_UI_Icon_Instrument_Type_SDF
graph {
 properties {blendMode=BLEND_Translucent, materialDomain=MD_UI, pixelDepthOffsetMode=PDOM_Legacy, usedShadingModels=Default Lit}
 nodes [
  NamedRerouteDeclaration(None=MaterialFunctionCall(UVs=TextureCoordinate(), Texture Scale=ScalarParameter() {defaultValue=1, parameterName=Size}) {materialFunction=&ScaleUVsByCenter}.UVs) {name=TextureCoord, nodeColor=<rgba 1 0 0.3516 1>}
  n1 = Constant()
  n2 = Constant() {r=-0.25}
  n3 = MaterialFunctionCall(UVs=Add(A=NamedRerouteUsage(), B=AppendVector(A=@n1, B=@n2)) {constB=-0.5}, Texture Scale=NamedRerouteUsage()) {materialFunction=&ScaleUVsByCenter}
  NamedRerouteDeclaration(None=ChannelMaskParameter(None=TextureSample(UVs=@n3.UVs) {samplerSource=SSM_Clamp_WorldGroupSettings, texture=/FMJamSystem/Textures/Icons/T_Icon_Instruments_Type_SDF}.RGB) {maskChannel=Green, parameterName=Drums, defaultValue=<rgba 0 1 0 0>}.R) {name=AllParts_Drums, nodeColor=<rgba 0 0.6593 1 1>}
  n4 = NamedRerouteUsage()
  n5 = TextureSample(UVs=@n4) {samplerSource=SSM_Clamp_WorldGroupSettings, texture=/FMJamSystem/Textures/Icons/T_Icon_Instruments_Type_2_SDF}
  NamedRerouteDeclaration(None=@n5.G) {name=Texture2_Green, nodeColor=<rgba 0.1054 1 0 1>}
  n6 = NamedRerouteUsage()
  n7 = NamedRerouteUsage()
  n8 = NamedRerouteUsage()
  n9 = NamedRerouteUsage()
  n10 = MaterialFunctionCall(DistanceField=Add(A=Add(A=@n6, B=@n7), B=Add(A=@n8, B=@n9)), Threshold=Constant() {r=0.32}, Smoothing=Constant() {r=1}) {materialFunction=&MF_UI_AntiAliasing}
  NamedRerouteDeclaration(None=Subtract(A=NamedRerouteUsage(), B=Saturate(None=@n10.Outer))) {name=AllPartsCombined, nodeColor=<rgba 0.2891 0 1 1>}
  n11 = ScalarParameter() {defaultValue=5, parameterName=Icon}
  n12 = LinearInterpolate(A=NamedRerouteUsage(), B=NamedRerouteUsage(), Alpha=Saturate(None=@n11))
  n13 = NamedRerouteUsage()
  n14 = Saturate(None=Subtract(A=@n11))
  n15 = Subtract(A=@n11) {constB=2}
  n16 = LinearInterpolate(
   A=LinearInterpolate(A=LinearInterpolate(A=@n12, B=@n13, Alpha=@n14), B=NamedRerouteUsage(), Alpha=Saturate(None=@n15))
   B=NamedRerouteUsage()
   Alpha=Saturate(None=Subtract(A=@n11) {constB=3})
  )
  n17 = NamedRerouteUsage()
  n18 = Saturate(None=Subtract(A=@n11) {constB=4})
  n19 = Subtract(A=NamedRerouteUsage()) {constB=5}
  NamedRerouteDeclaration(None=LinearInterpolate(A=LinearInterpolate(A=@n16, B=@n17, Alpha=@n18), B=NamedRerouteUsage(), Alpha=Saturate(None=@n19))) {name=PartsIndex, nodeColor=<rgba 0.0312 1 0 1>}
  NamedRerouteDeclaration(None=LinearInterpolate(A=Constant() {r=0.3}, B=Constant() {r=0.38}, Alpha=ScalarParameter() {defaultValue=1, parameterName=Size})) {name=SmallIconScale, nodeColor=<rgba 0.6875 0 1 1>}
  n20 = MaterialFunctionCall(UVs=TextureCoordinate(), "Rotation Angle (0-1)"=Constant() {r=0.25}) {materialFunction=&CustomRotator}
  n21 = ScalarParameter() {defaultValue=0.05, parameterName=AnimationSpeed}
  n22 = MaterialFunctionCall!(
   Value=ComponentMask(None=Panner(Coordinate=@n20.Rotated Values, Speed=@n21) {speedX=0.3, speedY=1}) {r=true}
   Scale=ScalarParameter() {defaultValue=24, parameterName=Lines}
   Height
  ) {materialFunction=/FortUILibrary/MaterialFunctions/Basics/MF_UI_PingPong}
  n23 = Constant() {r=0.5}
  n24 = LinearInterpolate(
   A=VectorParameter() {parameterName=QM_CL1, defaultValue=<rgba 0.885 0.956 0.1593 1>}.RGB
   B=VectorParameter() {parameterName=QM_CL2, defaultValue=<rgba 0.956 0.7158 0.1399 1>}.RGB
   Alpha=Saturate(None=MaterialFunctionCall(DistanceField=@n22.Result, Threshold=@n23) {materialFunction=&MF_UI_AntiAliasing}.Inner)
  )
  n25 = MaterialFunctionCall(Base=@n24, Blend=MaterialFunctionCall() {materialFunction=&LinearGradient}.VGradient) {materialFunction=&Blend_Overlay}
  n26 = LinearInterpolate(B=Multiply(A=@n25.Result) {constB=3}, Alpha=NamedRerouteUsage())
  NamedRerouteDeclaration(None=LinearInterpolate(A=LinearInterpolate(A=@n24, B=@n25.Result), Alpha=Saturate(None=@n26))) {name=ColourLines, nodeColor=<rgba 0 1 0.8125 1>}
  n27 = TextureCoordinate()
  n28 = Constant() {r=0.94}
  n29 = ScalarParameter() {defaultValue=0.03, parameterName=Outline}
  NamedRerouteDeclaration(None=Saturate(None=MaterialFunctionCall(UVs=@n27, Size=@n28, OutlineThickness=@n29) {materialFunction=&MF_UI_Circle}.Outline)) {name=outline, nodeColor=<rgba 0 0 0 1>}
  n30 = Constant()
  n31 = Constant() {r=0.25}
  n32 = MaterialFunctionCall(UVs=Add(A=NamedRerouteUsage(), B=AppendVector(A=@n30, B=@n31)) {constB=-0.5}, Texture Scale=NamedRerouteUsage()) {materialFunction=&ScaleUVsByCenter}
  NamedRerouteDeclaration(None=ChannelMaskParameter(None=TextureSample(UVs=@n32.UVs) {samplerSource=SSM_Clamp_WorldGroupSettings, texture=&T_Icon_Instruments_Type_2_SDF}.RGB) {maskChannel=Green, parameterName=Vocals, defaultValue=<rgba 0 1 0 0>}.R) {name=AllParts_Vocals, nodeColor=<rgba 1 0.7527 0 1>}
  n33 = Constant() {r=0.25}
  n34 = Constant()
  n35 = MaterialFunctionCall(UVs=Add(A=NamedRerouteUsage(), B=AppendVector(A=@n33, B=@n34)) {constB=-0.5}, Texture Scale=NamedRerouteUsage()) {materialFunction=&ScaleUVsByCenter}
  NamedRerouteDeclaration(None=ChannelMaskParameter(None=TextureSample(UVs=@n35.UVs) {samplerSource=SSM_Clamp_WorldGroupSettings, texture=&T_Icon_Instruments_Type_SDF}.RGB) {parameterName=Bass}.R) {name=AllParts_Bass, nodeColor=<rgba 0 1 0.1124 1>}
  n36 = Constant() {r=-0.25}
  n37 = Constant()
  n38 = MaterialFunctionCall(UVs=Add(A=NamedRerouteUsage(), B=AppendVector(A=@n36, B=@n37)) {constB=-0.5}, Texture Scale=NamedRerouteUsage()) {materialFunction=&ScaleUVsByCenter}
  NamedRerouteDeclaration(None=ChannelMaskParameter(None=TextureSample(UVs=@n38.UVs) {samplerSource=SSM_Clamp_WorldGroupSettings, texture=&T_Icon_Instruments_Type_SDF}.RGB) {maskChannel=Blue, parameterName=Guitar, defaultValue=<rgba 0 0 1 0>}.R) {name=AllParts_Guitar, nodeColor=<rgba 1 0.0927 0.8225 1>}
  NamedRerouteDeclaration(None=MaterialFunctionCall(UVs=NamedRerouteUsage(), Size=Constant() {r=0.9}, GlowMin=Constant() {r=-0.07}) {materialFunction=&MF_UI_Circle}.Fill) {name=CircleCrop, nodeColor=<rgba 0.3594 1 0 1>}
  n39 = TextureSample(UVs=@n4) {samplerSource=SSM_Clamp_WorldGroupSettings, texture=&T_Icon_Instruments_Type_SDF}
  NamedRerouteDeclaration(None=@n39.R) {name=Texture1_Red, nodeColor=<rgba 1 0.0448 0.3564 1>}
  NamedRerouteDeclaration(None=@n39.G) {name=Texture1_Green, nodeColor=<rgba 0.4297 1 0 1>}
  NamedRerouteDeclaration(None=@n39.B) {name=Texture1_Blue, nodeColor=<rgba 0.0628 0.2603 1 1>}
  NamedRerouteDeclaration(None=@n5.R) {name=Texture2_Red, nodeColor=<rgba 1 0.0274 0 1>}
  StaticSwitchParameter(True=NamedRerouteUsage(), False=VectorParameter() {parameterName=Icon Color, defaultValue=<rgba 1 1 1 1>}.RGB) {parameterName=Lines}
  n40 = Saturate(None=MaterialFunctionCall(
    DistanceField=NamedRerouteUsage()
    Threshold=ScalarParameter() {defaultValue=0.4, parameterName=AntiAliasing Threshold}
    Smoothing=ScalarParameter() {defaultValue=0.25, parameterName=AntiAliasing Smoothing}
   ) {materialFunction=&MF_UI_AntiAliasing}.Outer)
  n41 = ScalarParameter() {group=Gradient Settings, parameterName=GradientStart, sortPriority=1}
  n42 = ScalarParameter() {group=Gradient Settings, parameterName=GradientEnd, sortPriority=2}
  n43 = ScalarParameter() {group=Gradient Settings, parameterName=GradientRotation, sortPriority=4}
  n44 = ScalarParameter() {group=Gradient Settings, parameterName=GradientOffset, sortPriority=3}
  n45 = VectorParameter() {
   group=Gradient Settings
   parameterName=GradientRotationCenter
   sortPriority=5
   channelNames {a="", b="", g=Center Y, r=Center X}
   defaultValue=<rgba 0.5 0.5 0 0>
  }
  n46 = LinearInterpolate(
   A=@n40
   B=Multiply(
    A=@n40
    B=MaterialFunctionCall(GradientStart=@n41, GradientEnd=@n42, RotationAngle=@n43, RotationCenter=@n44, GradientOffset=@n45) {materialFunction=&MF_UI_LinearGradient}.Result
   )
   Alpha=ScalarParameter() {group=Gradient Settings, parameterName=GradientIntensity}
  )
  StaticSwitchParameter(True=LinearInterpolate(A=NamedRerouteUsage(), Alpha=@n46), False=@n46) {parameterName=Lines}
 ]
}
